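# Editor utility Blueprint of shared asset fix functions for the Tools Wrkshp tools, plus a getter for an available-fixes map.
# Mesh fixes handle LODs, collision and Nanite, mostly through StaticMeshEditorSubsystem.
# Texture fixes set compression, grayscale sRGB, maximum size and power-of-two stretching, calling TransactObject before edits.
# Fixes return a success bool, though the collision-complexity, collision-LOD and platform-minimum-LOD fixes read Nanite settings for it.
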EditorUtilityBlueprint /Game/Blueprints/Blutilities/ToolsWrkshp/ScrTools/Shared/SharedFunctions/BP_ToolsWrkshp_SharedAssetFixes
graphs [
 Get Available Fixes {nodes [
   Get Available Fixes() {is=title K2Node_FunctionEntry}
   > Return Node!(execute, Fixes:Map of Strings to Strings=Get Available Fixes() -> (Available Fixes:Map of Strings to Strings) {is=title K2Node_VariableGet})
  ]}
 MeshFix_AddLODs {nodes [
   n1 = Get Mesh!() -> (Mesh:Static Mesh Object Reference)
   n2 = CheckoutLoadedAsset!(execute, AssetToCheckout:Object Reference=@n1) -> (then, ReturnValue:bool)
   > Branch(Condition=@n2)
   > SaveLoadedAsset(AssetToSave=@n1, bOnlyIfIsDirty=false)
   MeshFix_AddLODs!() -> (then, Mesh:Static Mesh Object Reference) {is=title K2Node_FunctionEntry}
   > n3 = Sequence!(execute) -> (then_0, then_1, then_2:Exec)
   > SetLODGroup!(
    execute
    self:Static Mesh Editor Subsystem Object Reference=Get StaticMeshEditorSubsystem()
    StaticMesh:Static Mesh Object Reference=Get Mesh()
    LODGroup:Name=Default
    bRebuildImmediately:bool=true
   ) -> (then, ReturnValue:bool)
   > SetLodScreenSizes!(
    execute
    self:Static Mesh Editor Subsystem Object Reference=Get StaticMeshEditorSubsystem()
    StaticMesh:Static Mesh Object Reference=Get Mesh()
    ScreenSizes:"Array of Float (single-precision)s"=Get LOD_ScreenSizes() -> (LOD_ScreenSizes:"Array of Float (single-precision)s")
   ) -> (then, ReturnValue:bool)
   > SetMinLODForQualityLevels!(
    execute
    self:Static Mesh Object Reference=Get Mesh()
    QualityLevelMinimumLODs:Map of EPerQualityLevels Enums to Integers=Get LOD_QualityDefaults() -> (LOD_QualityDefaults:Map of EPerQualityLevels Enums to Integers=Low)
    Default:int=-1
   ) -> (then)
   Reroute Node(InputPin=@n3.then_1)
   n4 = Get Mesh()
   Return Node!(execute=@n3.then_2, Fix Successful:bool="integer > integer"(A=GetNumLODs(self=@n4), B=2))
  ]}
 MeshFix_AddSimpleCollision_NDOP26 {nodes [
   n1 = MeshFix_AddSimpleCollision_NDOP26!() -> (then, Mesh:Static Mesh Object Reference) {is=title K2Node_FunctionEntry}
   > TransactObject(Object=@n1)
   > AddSimpleCollisions!(
    execute
    self:Static Mesh Editor Subsystem Object Reference=Get StaticMeshEditorSubsystem()
    StaticMesh:Static Mesh Object Reference=@n1
    ShapeType:EScriptCollisionShapeType Enum=NDOP26
   ) -> (then, ReturnValue:int=0)
   > n2 = GetSimpleCollisionFromStaticMesh!(execute, StaticMesh:Static Mesh Object Reference=@n1) -> (then, ReturnValue:Geometry Script Simple Collision Structure)
   > n3 = GetSimpleCollisionShapeCount!(execute, SimpleCollision:"Geometry Script Simple Collision Structure (by ref)"=@n2) -> (then, ReturnValue:int=0)
   > Return Node(Fix Successful="integer > integer"(A=@n3))
  ]}
 MeshFix_CollisionComplexity {nodes [
   n1 = MeshFix_CollisionComplexity!() -> (then, Mesh:Static Mesh Object Reference) {is=title K2Node_FunctionEntry}
   > n2 = GetEditorProperty!(execute:Exec, Object:Object Reference=@n1, PropertyName:Name=BodySetup) -> (then:Exec, PropertyValue:Object Reference, ReturnValue:bool)
   > SetEditorProperty(
    Object=@n2
    PropertyName=CollisionTraceFlag
    PropertyValue:ECollisionTraceFlag Enum=Get Collision Complexity() -> (Collision Complexity:ECollisionTraceFlag Enum=CTF_UseDefault)
   )
   Reroute Node()
   n3 = Get StaticMeshEditorSubsystem()
   > Return Node(Fix Successful=Break Mesh Nanite Settings!(MeshNaniteSettings:"Mesh Nanite Settings Structure (by ref)"=GetNaniteSettings!(self:Static Mesh Editor Subsystem Object Reference=@n3, StaticMesh:Static Mesh Object Reference) -> (ReturnValue:Mesh Nanite Settings Structure)) -> (
     bEnabled:bool=~
     bExplicitTangents:bool=~
     bLerpUVs:bool=~
     bSeparable:bool=~
     bVoxelNDF:bool=~
     bVoxelOpacity:bool=~
     ShapePreservation:ENaniteShapePreservation Enum
     PositionPrecision:int
     NormalPrecision:int
     TangentPrecision:int
     BoneWeightPrecision:int
     KeepPercentTriangles:float=~
     TrimRelativeError:float=~
     GenerateFallback:ENaniteGenerateFallback Enum
     FallbackTarget:ENaniteFallbackTarget Enum
     FallbackPercentTriangles:float=~
     FallbackRelativeError:float=~
     MaxEdgeLengthFactor:float=~
     NumRays:int
     VoxelLevel:int
     RayBackUp:float=~
     DisplacementUVChannel:int
    ) {is=title K2Node_BreakStruct})
  ]}
 MeshFix_SetLODforCollision {nodes [
   n1 = MeshFix_SetLODforCollision!() -> (then, Mesh:Static Mesh Object Reference) {is=title K2Node_FunctionEntry}
   > Set LODForCollision!(execute, LODForCollision:int=int - int(A=GetNumLODs(self=@n1)), self:Static Mesh Object Reference=@n1) -> (then, Output_Get:int=0)
   Reroute Node()
   n2 = Get StaticMeshEditorSubsystem()
   > Return Node(Fix Successful=Break Mesh Nanite Settings(MeshNaniteSettings=GetNaniteSettings(self=@n2)) -> (
     bEnabled=~
     bExplicitTangents=~
     bLerpUVs=~
     bSeparable=~
     bVoxelNDF=~
     bVoxelOpacity=~
     KeepPercentTriangles=~
     TrimRelativeError=~
     FallbackPercentTriangles=~
     FallbackRelativeError=~
     MaxEdgeLengthFactor=~
     RayBackUp=~
    ))
  ]}
 MeshFix_SetMinLODforPlatform {nodes [
   MeshFix_SetMinLODforPlatform!() -> (then, Mesh:Static Mesh Object Reference) {is=title K2Node_FunctionEntry}
   > SetMinLODForQualityLevels(
    self=Get Mesh()
    QualityLevelMinimumLODs=Get L_Env_LODQuality_Defaults() -> (L_Env_LODQuality_Defaults:Map of EPerQualityLevels Enums to Integers=Low)
    Default=-1
   )
   Reroute Node()
   n1 = Get StaticMeshEditorSubsystem()
   > Return Node(Fix Successful=Break Mesh Nanite Settings(MeshNaniteSettings=GetNaniteSettings(self=@n1)) -> (
     bEnabled=~
     bExplicitTangents=~
     bLerpUVs=~
     bSeparable=~
     bVoxelNDF=~
     bVoxelOpacity=~
     KeepPercentTriangles=~
     TrimRelativeError=~
     FallbackPercentTriangles=~
     FallbackRelativeError=~
     MaxEdgeLengthFactor=~
     RayBackUp=~
    ))
  ]}
 MeshFix_TurnOnNanite {nodes [
   MeshFix_TurnOnNanite!() -> (then, Mesh:Static Mesh Object Reference) {is=title K2Node_FunctionEntry}
   > n1 = Sequence!(execute) -> (then_0, then_1)
   n2 = Get StaticMeshEditorSubsystem()
   n3 = Get Mesh()
   Return Node(
    execute=@n1.then_1
    Fix Successful=Break Mesh Nanite Settings(MeshNaniteSettings=GetNaniteSettings(self=@n2, StaticMesh=@n3)) -> (
     bEnabled=~
     bExplicitTangents=~
     bLerpUVs=~
     bSeparable=~
     bVoxelNDF=~
     bVoxelOpacity=~
     KeepPercentTriangles=~
     TrimRelativeError=~
     FallbackPercentTriangles=~
     FallbackRelativeError=~
     MaxEdgeLengthFactor=~
     RayBackUp=~
    )
   )
   > SetNaniteSettings!(
    execute
    self:Static Mesh Editor Subsystem Object Reference=Get StaticMeshEditorSubsystem()
    StaticMesh:Static Mesh Object Reference=Get Mesh()
    NaniteSettings:Mesh Nanite Settings Structure=Make Mesh Nanite Settings!(
     bEnabled:bool=true
     bExplicitTangents:bool=true
     bLerpUVs:bool=True
     bSeparable:bool=False
     bVoxelNDF:bool=True
     bVoxelOpacity:bool=False
     ShapePreservation:ENaniteShapePreservation Enum=None
     PositionPrecision:int=-2147483648
     NormalPrecision:int=-1
     TangentPrecision:int=-1
     BoneWeightPrecision:int=-1
     KeepPercentTriangles:float=1
     TrimRelativeError:float=0
     GenerateFallback:ENaniteGenerateFallback Enum=PlatformDefault
     FallbackTarget:ENaniteFallbackTarget Enum=RelativeError
     FallbackPercentTriangles:float=1
     FallbackRelativeError:float=0
     MaxEdgeLengthFactor:float=0
     NumRays:int=64
     VoxelLevel:int=0
     RayBackUp:float=0
     DisplacementUVChannel:int=0
    ) -> (MeshNaniteSettings:Mesh Nanite Settings Structure) {is=title K2Node_MakeStruct}
    bApplyChanges:bool=true
   ) -> (then)
  ]}
 TextureFix_Compression_GrayscaleSRGB {nodes [
   n1 = TextureFix_Compression_GrayscaleSRGB!() -> (then, Texture:Texture 2D Object Reference) {is=title K2Node_FunctionEntry}
   > TransactObject(Object=@n1)
   > Set SRGB(SRGB=false, self=@n1)
   > Return Node(Fix Successful=NOT Boolean(A=Get SRGB!(self:Texture Object Reference=@n1) -> (SRGB:bool)))
   n2 = Branch(Condition="Equal (Enum)"(A:TextureCompressionSettings Enum=Get CompressionSettings(), B:TextureCompressionSettings Enum=TC_Grayscale))
   Return Node(execute=@n2.else, Fix Successful=false)
  ]}
 TextureFix_Compression_SetToBC7 {nodes [
   n1 = TextureFix_Compression_SetToBC7!() -> (then, Texture:Texture 2D Object Reference) {is=title K2Node_FunctionEntry}
   > TransactObject(Object=@n1)
   > Set CompressionSettings(CompressionSettings=TC_BC7, self=@n1)
   > Return Node(Fix Successful="Equal (Enum)"(A=Get CompressionSettings(self=@n1), B=TC_BC7))
  ]}
 TextureFix_GrayscaleSRGB {nodes [
   n1 = TextureFix_GrayscaleSRGB!() -> (then, Texture:Texture 2D Object Reference) {is=title K2Node_FunctionEntry}
   > n2 = Branch(Condition="Equal (Enum)"(A=Get CompressionSettings(self=@n1), B=TC_Grayscale))
   > TransactObject(Object=@n1)
   > Set SRGB(SRGB=false, self=@n1)
   > Return Node(Fix Successful=NOT Boolean(A=Get SRGB(self=@n1)))
   Return Node(execute=@n2.else, Fix Successful=false)
  ]}
 TextureFix_MaxSize {nodes [
   n1 = TextureFix_MaxSize!() -> (then, Texture:Texture 2D Object Reference, Maximum Texture Size:int=2048) {is=title K2Node_FunctionEntry}
   > TransactObject(Object=@n1)
   > Set MaxTextureSize!(execute, MaxTextureSize:int=@n1.Maximum Texture Size, self:Texture Object Reference=@n1) -> (then, Output_Get:int=0)
   n2 = GetSizeY!(self:Texture 2D Object Reference=@n1) -> (ReturnValue:int=0)
   n3 = Get Maximum Texture Size() -> (Maximum Texture Size:int=0)
   n4 = GetSizeX!(self:Texture 2D Object Reference=@n1) -> (ReturnValue:int=0)
   > Return Node(Fix Successful=AND Boolean(A="integer <= integer"(A=@n4, B=@n3), B="integer <= integer"(A=@n2, B=@n3)))
  ]}
 TextureFix_PowerOf2 {nodes [
   n1 = TextureFix_PowerOf2!() -> (then, Texture:Texture 2D Object Reference) {is=title K2Node_FunctionEntry}
   > PrintText(InText=Format Text!(Format="LOCTEXT(\":\\r\\nStretch to Pow 2: {a}\")", a:Texture 2D Object Reference=@n1) -> (Result))
   > TransactObject(Object=@n1)
   > Set PowerOfTwoMode!(execute, PowerOfTwoMode:ETexturePowerOfTwoSetting Enum=StretchToPowerOfTwo, self:Texture Object Reference=@n1) -> (then, Output_Get:ETexturePowerOfTwoSetting Enum=None)
   > Return Node(Fix Successful="Equal (Enum)"(
     A:ETexturePowerOfTwoSetting Enum=Get PowerOfTwoMode!(self:Texture Object Reference=@n1) -> (PowerOfTwoMode:ETexturePowerOfTwoSetting Enum=None)
     B:ETexturePowerOfTwoSetting Enum=StretchToPowerOfTwo
    ))
   BeginTransaction(Context=Tools Wrkshp, Description="LOCTEXT(\"Tools Wrkshp - Set Padding and Resizing\")")
   EndTransaction()
  ]}
]
variables []
