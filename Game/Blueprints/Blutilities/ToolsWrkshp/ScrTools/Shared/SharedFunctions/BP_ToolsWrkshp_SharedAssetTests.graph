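# Editor utility library of shared asset health checks for a tools workshop, flagging problems in sound waves, static meshes and textures.
# Each check tests properties like channels, LODs, collision, Nanite or compression, then returns a map of assets to errors, updated on failure.
# Mesh and texture checks attach tooltips looked up by check name.
# Many accept an Only Major filter, gated by size-scaled vertex thresholds for meshes and resolution for textures.
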EditorUtilityBlueprint /Game/Blueprints/Blutilities/ToolsWrkshp/ScrTools/Shared/SharedFunctions/BP_ToolsWrkshp_SharedAssetTests
graphs [
 AudioCheck_Channels {nodes [
   n1 = AudioCheck_Channels!() -> (
    then
    Target:Sound Wave Object Reference
    Acceptable Channel Count:int=2
    ChannelsArray:Map of Sound Wave Object References to S Asset Errors Asset Health Structures
   ) {is=title K2Node_FunctionEntry}
   > Set CurSoundWave(CurSoundWave:Sound Wave Object Reference=@n1) -> (Output_Get:Sound Wave Object Reference)
   > n2 = Branch(Condition=NOT Boolean(A="integer <= integer"(B=@n1.Acceptable Channel Count)))
   > n3 = Update SoundWave Error Array!(
    execute
    The Sound Wave:Sound Wave Object Reference=Get CurSoundWave() -> (CurSoundWave:Sound Wave Object Reference)
    The Array:Map of Sound Wave Object References to S Asset Errors Asset Health Structures=Get ChannelsArray() -> (ChannelsArray:Map of Sound Wave Object References to S Asset Errors Asset Health Structures)
    The Error:Text=Format Text!(
     Format="LOCTEXT(\"Too many channels: {l}\\r\\nUseful max channels is {ul}\")"
     l:double="To Float (Integer)"(InInt=0)
     ul:int=@n1.Acceptable Channel Count
    ) -> (Result)
    __WorldContext:Object Reference
   ) -> (then, Updated Array:Map of Sound Wave Object References to S Asset Errors Asset Health Structures)
   > Return Node!(execute, Updated Array:Map of Sound Wave Object References to S Asset Errors Asset Health Structures=@n3)
   Return Node(execute=@n2.else, Updated Array=Get ChannelsArray())
  ]}
 AudioCheck_HighSample {nodes [
   n1 = AudioCheck_HighSample!() -> (
    then
    Target:Sound Wave Object Reference
    Acceptable Sample Rate:double=32000
    HighSampleArray:Map of Sound Wave Object References to S Asset Errors Asset Health Structures
   ) {is=title K2Node_FunctionEntry}
   > Set CurSoundWave(CurSoundWave=@n1)
   n2 = SafeDivide(
    A=Get TotalSamples!(self:Sound Base Object Reference=@n1) -> (TotalSamples:float)
    B=Get Duration!(self:Sound Base Object Reference=@n1) -> (Duration:float)
   )
   > n3 = Branch(Condition=NOT Boolean(A="float <= float"(A=@n2, B=@n1.Acceptable Sample Rate) -> (ReturnValue=~)))
   > n4 = Update SoundWave Error Array(
    The Sound Wave=Get CurSoundWave()
    The Array=Get HighSampleArray() -> (HighSampleArray:Map of Sound Wave Object References to S Asset Errors Asset Health Structures)
    The Error=Format Text!(
     Format="LOCTEXT(\"Too high sample rate {r}\\r\\nUseful max sample rate {ms}\")"
     r:int=Round(A=@n2) {is=title K2Node_CallFunction}
     ms:double=Get Acceptable Sample Rate() -> (Acceptable Sample Rate:double)
    ) -> (Result)
   )
   > Return Node(Updated Array=@n4)
   Return Node(execute=@n3.else, Updated Array=Get HighSampleArray())
  ]}
 AudioCheck_Long {nodes [
   n1 = AudioCheck_Long!() -> (
    then
    Target:Sound Wave Object Reference
    Acceptable Length:double=30
    LongArray:Map of Sound Wave Object References to S Asset Errors Asset Health Structures
   ) {is=title K2Node_FunctionEntry}
   > Set CurSoundWave(CurSoundWave=@n1)
   n2 = Get Duration(self=@n1)
   > n3 = Branch(Condition=NOT Boolean(A="float <= float"(A=@n2, B=@n1.Acceptable Length) -> (ReturnValue=~)))
   > n4 = Update SoundWave Error Array(
    The Sound Wave=Get CurSoundWave()
    The Array=Get LongArray() -> (LongArray:Map of Sound Wave Object References to S Asset Errors Asset Health Structures)
    The Error=Format Text!(Format="LOCTEXT(\"Too long: {l} seconds\\r\\nUseful max length is {ul} seconds\")", l:float=@n2, ul:double=@n1.Acceptable Length) -> (Result)
   )
   > Return Node(Updated Array=@n4)
   Return Node(execute=@n3.else, Updated Array=Get LongArray())
  ]}
 Get Mesh Check Fix Tooltips {nodes [
   n1 = Get Mesh Check Fix Tooltips() {is=title K2Node_FunctionEntry}
   n2 = Get Mesh Fix Tooltips() -> (Mesh Fix Tooltips:Map of Strings to Texts)
   Return Node!(execute=[@n1.then, @n3.CompletedPin], MeshCheck_NumMatSlots:Map of Strings to Texts=@n2)
   n3 = "For Each Loop (Map)"!(execute:Exec, MapPin:Map of Strings to Texts=@n2, BreakPin:Exec) -> (then:Exec, KeyPin:String, ValuePin:Text, CompletedPin:Exec)
   > PrintText(InText=Format Text!(Format="LOCTEXT(\":\\r\\n:\\r\\nCheck Name: {c}\\r\\n\\r\\nTooltip: {t}\\r\\n------------\")", c:String=@n3, t:Text=@n3.ValuePin) -> (Result))
  ]}
 Get Texture Check Fix Tooltips {nodes [
   n1 = Get Texture Check Fix Tooltips() {is=title K2Node_FunctionEntry}
   n2 = Get Texture Fix Tooltips() -> (Texture Fix Tooltips:Map of Strings to Texts)
   Return Node(execute=[@n1.then, @n3.CompletedPin], MeshCheck_NumMatSlots=@n2)
   n3 = "For Each Loop (Map)"(MapPin=@n2)
   > PrintText(InText=Format Text(Format="LOCTEXT(\":\\r\\n:\\r\\nCheck Name: {c}\\r\\n\\r\\nTooltip: {t}\\r\\n------------\")", c=@n3, t=@n3.ValuePin))
  ]}
 IsMajorIssue_Mesh {nodes [
   n1 = IsMajorIssue_Mesh!() -> (then, StaticMesh:Static Mesh Object Reference, Only Major:bool=~) {is=title K2Node_FunctionEntry}
   > n2 = Branch(Condition=@n1.Only Major)
   n3 = Get Character Size() -> (Character Size:double)
   n4 = Make Map!(
    Key 0:double=Make Literal Float(Value=float * float(A=@n3, B=0.5) {is=title K2Node_PromotableOperator})
    Value 0:int=MakeLiteralInt!(Value:int=125) -> (ReturnValue:int=0)
    Key 1:double=Make Literal Float(Value=float * float(A=@n3, B=1))
    Value 1:int=MakeLiteralInt(Value=375) -> (ReturnValue=0)
    Key 2:double=Make Literal Float(Value=float * float(A=@n3, B=3))
    Value 2:int=MakeLiteralInt(Value=625) -> (ReturnValue=0)
   ) -> (Map:"Map of Float (double-precision)s to Integers") {is=title K2Node_MakeMap}
   > Set Major Issue Thresholds(Major Issue Thresholds:"Map of Float (double-precision)s to Integers"=@n4) -> (Output_Get:"Map of Float (double-precision)s to Integers")
   > n5 = Sequence()
   > Set Found Threshold(Found Threshold:bool=false) -> (Output_Get:bool)
   > n6 = "For Each Loop (Map)"(MapPin:"Map of Float (double-precision)s to Integers"=@n4, BreakPin=@n8.then) -> (KeyPin:double, ValuePin:int)
   n7 = "Get Size (Box)"!(InBox:"Box Structure (by ref)"=GetBoundingBox!(self:Static Mesh Object Reference=@n1) -> (ReturnValue:Box Structure)) -> (ReturnValue:Vector) {is=title K2Node_CallFunction}
   > Branch(Condition="float < float"(A=GetMaxElement!(A:Vector=@n7) -> (ReturnValue:double), B=@n6))
   > Set Found Threshold Size(Found Threshold Size:double=@n6) -> (Output_Get:double)
   > n8 = Set Found Threshold(Found Threshold=true)
   Branch(execute=@n6.CompletedPin, Condition=NOT Boolean(A=Get Found Threshold() -> (Found Threshold:bool)))
   n9 = Get Major Issue Thresholds() -> (Major Issue Thresholds:"Map of Float (double-precision)s to Integers")
   n10 = Length!(TargetMap:"Map of Float (double-precision)s to Integers"=@n9) -> (ReturnValue:int=0) {is=title K2Node_CallFunction}
   > Set Found Threshold Size(Found Threshold Size=Get Key Value By Index!(TargetMap:"Map of Float (double-precision)s to Integers"=@n9, Index:int=int - int(A=@n10)) -> (Key:double=~, Value:int) {is=title K2Node_CallFunction})
   n11 = Branch(execute=@n5.then_1)
   > n12 = Compare Int!(
    Exec:Exec
    Input:int=GetNumberVerts!(
     self:Static Mesh Editor Subsystem Object Reference=Get StaticMeshEditorSubsystem()
     StaticMesh:Static Mesh Object Reference=@n1
     LODIndex:int=0
    ) -> (ReturnValue:int=0)
    CompareWith:int=Find(
     TargetMap:"Map of Float (double-precision)s to Integers"=Get Major Issue Thresholds()
     Key:"Float (double-precision) (by ref)"=Get Found Threshold Size() -> (Found Threshold Size:double)
    ) -> (Value:int)
   ) -> (">":Exec, "==":Exec, "<":Exec) {is=title K2Node_MacroInstance}
   Return Node!(execute=[@n12."<", @n12."=="], Continue Test:bool=false)
   > Return Node(Continue Test=true)
   Return Node(execute=@n11.else, Continue Test=true)
   Return Node(execute=@n2.else, Continue Test=true)
   n13 = GetNumberVerts(self=Get StaticMeshEditorSubsystem(), LODIndex=0) -> (ReturnValue=0)
   PrintString(InString="To String (Text)"(InText=Format Text!(Format="LOCTEXT(\"Mesh has {1} Vertices. \")", 1:int=@n13) -> (Result)))
   n14 = Get Found Threshold Size()
   n15 = Find(TargetMap=Get Major Issue Thresholds(), Key=@n14)
   PrintString(InString="To String (Text)"(InText=Format Text!(Format="LOCTEXT(\"Threshold found at size {1} with {2} Vertices. \")", 1:double=@n14, 2:int=@n15) -> (Result)))
   Return Node(Continue Test=true)
  ]}
 IsMajorIssue_Texture {nodes [
   n1 = IsMajorIssue_Texture!() -> (then, Texture:Texture 2D Object Reference, Only Major:bool=~, Major Texture Threshold:int) {is=title K2Node_FunctionEntry}
   > n2 = Branch(Condition=@n1.Only Major)
   n3 = GetSizeY!(self:Texture 2D Object Reference=@n1) -> (ReturnValue:int=0)
   n4 = GetSizeX!(self:Texture 2D Object Reference=@n1) -> (ReturnValue:int=0)
   > n5 = Branch(Condition=AND Boolean(A="integer > integer"(A=@n4, B=@n1.Major Texture Threshold), B="integer > integer"(A=@n3, B=@n1.Major Texture Threshold)))
   Return Node(execute=[@n5.then, @n2.else], Continue Test=true)
   Return Node(execute=@n5.else, Continue Test=false)
  ]}
 IsPow2 {nodes [
   n1 = IsPow2!() -> (then, Input Value:int) {is=title K2Node_FunctionEntry}
   > n2 = Branch(Condition="integer > integer"(A=@n1))
   n3 = int - int(A=Get Input Value() -> (Input Value:int=0))
   n4 = Get Input Value() -> (Input Value=0)
   > Return Node!(
    execute
    Is Pow 2:bool="Equal (Integer)"(A=Bitwise AND(A:int=@n3, B:int=@n4) -> (ReturnValue:int=0) {is=title K2Node_CommutativeAssociativeBinaryOperator})
   )
   Return Node(execute=@n2.else, Is Pow 2=false)
  ]}
 MeshCheck_CollisionOnSections {nodes [
   MeshCheck_CollisionOnSections!() -> (
    then
    Static Mesh:Static Mesh Object Reference
    CollisionOnSectionsArray:Map of Static Mesh Object References to S Asset Errors Asset Health Structures
   ) {is=title K2Node_FunctionEntry}
   > Set l_hadASectionWithCollision(l_hadASectionWithCollision:bool=false) -> (Output_Get:bool)
   n1 = Get Static Mesh() -> (Static Mesh:Static Mesh Object Reference)
   > n2 = For Loop with Break!(
    Execute:Exec
    FirstIndex:int
    LastIndex:int=int - int(A=GetNumSections!(self:Static Mesh Object Reference=@n1, InLOD:int=0) -> (ReturnValue:int=0), B=1) {is=title K2Node_CallFunction}
    Break:Exec=@n3.then
   ) -> (LoopBody:Exec, Index:int, Completed:Exec) {is=title K2Node_MacroInstance}
   Return Node!(execute=@n2.Completed, hasSectionWithCollision:bool=Get l_hadASectionWithCollision() -> (l_hadASectionWithCollision:bool))
   > Branch(Condition=IsSectionCollisionEnabled!(
     self:Static Mesh Editor Subsystem Object Reference=Get StaticMeshEditorSubsystem()
     StaticMesh:Static Mesh Object Reference=Get Static Mesh()
     LODIndex:int=0
     SectionIndex:int=@n2
    ) -> (ReturnValue:bool))
   > n3 = Set l_hadASectionWithCollision(l_hadASectionWithCollision=true)
  ]}
 MeshCheck_ComplexAsSimple {nodes [
   n1 = MeshCheck_ComplexAsSimple!() -> (
    then
    Static Mesh:Static Mesh Object Reference
    ComplexCollisionArray:Map of Static Mesh Object References to S Asset Errors Asset Health Structures
    Only Major:bool=~
   ) {is=title K2Node_FunctionEntry}
   > n2 = Branch(Condition=IsMajorIssue_Mesh!(StaticMesh:Static Mesh Object Reference=@n1, Only Major:bool=@n1.Only Major, __WorldContext:Object Reference) -> (Continue Test:bool) {is=title K2Node_CallFunction})
   n3 = GetCollisionComplexity!(self:Static Mesh Editor Subsystem Object Reference=Get StaticMeshEditorSubsystem(), StaticMesh:Static Mesh Object Reference=@n1) -> (ReturnValue:ECollisionTraceFlag Enum=CTF_UseDefault)
   > n4 = Branch(Condition=NOT Boolean(A="Equal (Enum)"(A:ECollisionTraceFlag Enum=@n3, B:ECollisionTraceFlag Enum=CTF_UseComplexAsSimple)))
   > Return Node!(
    execute
    Updated Array:Map of Static Mesh Object References to S Asset Errors Asset Health Structures=Get ComplexCollisionArray() -> (ComplexCollisionArray:Map of Static Mesh Object References to S Asset Errors Asset Health Structures)
   )
   n5 = Get Mesh Check Fix Tooltips!(execute=@n4.else, __WorldContext:Object Reference) -> (then, MeshCheck_NumMatSlots:Map of Strings to Texts) {is=title K2Node_CallFunction}
   > n6 = Update StaticMesh Error Array!(
    execute
    The Static Mesh:Static Mesh Object Reference=Get Static Mesh()
    The Array:Map of Static Mesh Object References to S Asset Errors Asset Health Structures=Get ComplexCollisionArray()
    The Error:Text=Format Text!(Format="LOCTEXT(\"Complex as Simple Collision\")") -> (Result)
    The Fix if Any:String
    Tooltip:Text=Find(TargetMap:Map of Strings to Texts=@n5, Key:"String (by ref)"=MeshCheck_ComplexAsSimple) -> (Value:Text)
    __WorldContext:Object Reference
   ) -> (then, Updated Array:Map of Static Mesh Object References to S Asset Errors Asset Health Structures)
   > Return Node(Updated Array=@n6)
   Return Node(execute=@n2.else, Updated Array=Get ComplexCollisionArray())
   MakeLiteralString(Value=MeshFix_CollisionComplexity)
  ]}
 MeshCheck_HasLODs {nodes [
   n1 = MeshCheck_HasLODs!() -> (
    then
    Static Mesh:Static Mesh Object Reference
    MinLODCount:int=2
    HasLODS_Array:Map of Static Mesh Object References to S Asset Errors Asset Health Structures
    Only Major:bool=~
   ) {is=title K2Node_FunctionEntry}
   > n2 = Branch(Condition=IsMajorIssue_Mesh(StaticMesh=@n1, Only Major=@n1.Only Major))
   n3 = GetNumLODs(self=@n1)
   > n4 = Branch(Condition="integer < integer"(A=@n3, B=@n1.MinLODCount))
   > n5 = Get Mesh Check Fix Tooltips()
   > n6 = Update StaticMesh Error Array(
    The Static Mesh=Get Static Mesh()
    The Array=Get HasLODS_Array() -> (HasLODS_Array:Map of Static Mesh Object References to S Asset Errors Asset Health Structures)
    The Error="LOCTEXT(\"Too Few LODs\")"
    Tooltip=Format Text!(
     Format="LOCTEXT(\"{0}\\r\\n\\r\\n{1}\")"
     0:Text=Format Text!(Format="LOCTEXT(\"Number of LODs: {1}\\r\\nMinimum recommended is: {2}\")", 1:int=@n3, 2:int=@n1.MinLODCount) -> (Result)
     1:Text=Find(TargetMap=@n5, Key=MeshCheck_HasLODs)
    ) -> (Result)
   )
   > Return Node(Updated Array=@n6)
   Return Node(execute=@n4.else, Updated Array=Get HasLODS_Array())
   Return Node(execute=@n2.else, Updated Array=Get HasLODS_Array())
   MakeLiteralString(Value=MeshFix_AddLODs)
  ]}
 MeshCheck_LOD0_Collision {nodes [
   n1 = MeshCheck_LOD0_Collision!() -> (
    then
    Static Mesh:Static Mesh Object Reference
    LOD0_CollisionArray:Map of Static Mesh Object References to S Asset Errors Asset Health Structures
    Only Major:bool=~
   ) {is=title K2Node_FunctionEntry}
   > n2 = Branch(Condition=IsMajorIssue_Mesh(StaticMesh=@n1, Only Major=@n1.Only Major))
   n3 = Get LODForCollision!(self:Static Mesh Object Reference=@n1) -> (LODForCollision:int=0)
   > n4 = Branch(Condition="integer > integer"(A=@n3, B=1))
   > Return Node(Updated Array=Get LOD0_CollisionArray() -> (LOD0_CollisionArray:Map of Static Mesh Object References to S Asset Errors Asset Health Structures))
   n5 = Get Mesh Check Fix Tooltips(execute=@n4.else)
   > n6 = Update StaticMesh Error Array(
    The Static Mesh=Get Static Mesh()
    The Array=Get LOD0_CollisionArray()
    The Error="LOCTEXT(\"LOD for Collision\")"
    Tooltip=Format Text!(
     Format="LOCTEXT(\"{0}\\r\\n\\r\\n{1}\")"
     0:Text=Format Text!(Format="LOCTEXT(\"LOD for Collision is: {1}\\r\\nRecommended is higher than 1\")", 1:int=@n3) -> (Result)
     1:Text=Find(TargetMap=@n5, Key=MeshCheck_LOD0_Collision)
    ) -> (Result)
   )
   > Return Node(Updated Array=@n6)
   Return Node(execute=@n2.else, Updated Array=Get LOD0_CollisionArray())
   "integer < integer"!(A:int=@n3, B:int=1) -> (ReturnValue:bool) {is=title K2Node_CallFunction}
   MakeLiteralString(Value=MeshFix_SetLODforCollision)
  ]}
 MeshCheck_ManyCollPrims {nodes [
   n1 = MeshCheck_ManyCollPrims!() -> (
    then
    Static Mesh:Static Mesh Object Reference
    AcceptableCollPrims:int=5
    ManyCollPrims_Array:Map of Static Mesh Object References to S Asset Errors Asset Health Structures
    Only Major:bool=~
   ) {is=title K2Node_FunctionEntry}
   > n2 = Branch(Condition=IsMajorIssue_Mesh(StaticMesh=@n1, Only Major=@n1.Only Major))
   n3 = Get StaticMeshEditorSubsystem()
   n4 = Get Static Mesh()
   > n5 = GetSimpleCollisionCount!(execute, self:Static Mesh Editor Subsystem Object Reference=@n3, StaticMesh:Static Mesh Object Reference=@n4) -> (then, ReturnValue:int=0)
   > n6 = GetConvexCollisionCount!(execute, self:Static Mesh Editor Subsystem Object Reference=@n3, StaticMesh:Static Mesh Object Reference=@n4) -> (then, ReturnValue:int=0)
   n7 = "Max (Integer)"(A=@n6, B=@n5)
   > n8 = Branch(Condition="integer > integer"(A=@n7, B=Get AcceptableCollPrims() -> (AcceptableCollPrims:int=0)))
   > n9 = Get Mesh Check Fix Tooltips()
   n10 = Get AcceptableCollPrims() -> (AcceptableCollPrims=0)
   > n11 = Update StaticMesh Error Array(
    The Static Mesh=Get Static Mesh()
    The Array=Get ManyCollPrims_Array() -> (ManyCollPrims_Array:Map of Static Mesh Object References to S Asset Errors Asset Health Structures)
    The Error="LOCTEXT(\"Too Many Collision Prims\")"
    Tooltip=Format Text!(
     Format="LOCTEXT(\"{0}\\r\\n\\r\\n{1}\")"
     0:Text=Format Text(Format="LOCTEXT(\"Number of collision primitives: {0}\\r\\nShould be {1} or less if possible.\")", 0:int=@n7, 1:int=@n10)
     1:Text=Find(TargetMap=@n9, Key=MeshCheck_ManyCollPrims)
    ) -> (Result)
   )
   > Return Node(Updated Array=@n11)
   Return Node(execute=@n8.else, Updated Array=Get ManyCollPrims_Array())
   Return Node(execute=@n2.else, Updated Array=Get ManyCollPrims_Array())
  ]}
 MeshCheck_MinLODforLow {nodes [
   n1 = MeshCheck_MinLODforLow!() -> (
    then
    Static Mesh:Static Mesh Object Reference
    MinLOD_ForLow:int=2
    MinLOD_ForMedium:int=2
    MinLODforLowArray:Map of Static Mesh Object References to S Asset Errors Asset Health Structures
    Only Major:bool=~
   ) {is=title K2Node_FunctionEntry}
   > n2 = Branch(Condition=IsMajorIssue_Mesh(StaticMesh=@n1, Only Major=@n1.Only Major))
   > Set HasError(HasError:bool=false) -> (Output_Get:bool)
   > n3 = Sequence!(execute) -> (then_0, then_1, then_2:Exec)
   n4 = Branch(execute=@n3.then_2, Condition=Get HasError() -> (HasError:bool))
   > n5 = For Each Loop(Array:Array of Texts=Get Messages() -> (Messages:Array of Texts)) -> (Array Element:Text)
   n6 = Get Mesh Check Fix Tooltips(execute=@n5.Completed)
   > n7 = Update StaticMesh Error Array(
    The Static Mesh=Get Static Mesh()
    The Array=Get MinLODforLowArray() -> (MinLODforLowArray:Map of Static Mesh Object References to S Asset Errors Asset Health Structures)
    The Error="LOCTEXT(\"Min LOD for Platform\")"
    Tooltip=Format Text(
     Format="LOCTEXT(\"{0}\\r\\n\\r\\n{1}\")"
     0:Text=Get FinalErrors() -> (FinalErrors:Text)
     1:Text=Find(TargetMap=@n6, Key=MeshCheck_MinLODforLow)
    )
   )
   > Return Node(Updated Array=@n7)
   n8 = Get FinalErrors()
   n9 = Len(S="To String (Text)"(InText=@n8))
   Set FinalErrors(
    execute=@n5.LoopBody
    FinalErrors:Text=SelectText(
     A=Format Text!(Format="LOCTEXT(\"{a}\")", a:Text=@n5) -> (Result)
     B=Format Text!(Format="LOCTEXT(\"{a}\\r\\n{b}\")", a=@n8, b:Text=@n5) -> (Result)
     bPickA="Equal (Integer)"(A=@n9)
    )
   ) -> (Output_Get:Text)
   Return Node(execute=@n4.else, Updated Array=Get MinLODforLowArray())
   n10 = GetMinimumLODForQualityLevels!(self:Static Mesh Object Reference=Get Static Mesh()) -> (QualityLevelMinimumLODs:Map of Names to Integers)
   n11 = Find(TargetMap:Map of Names to Integers=@n10, Key:"Name (by ref)"=Low) -> (Value:int)
   n12 = Branch(execute=@n3.then_0, Condition=@n11.ReturnValue)
   > n13 = Set FoundMinLodForLow(FoundMinLodForLow:int=@n11) -> (Output_Get:int=0)
   n14 = Get MinLOD_ForLow() -> (MinLOD_ForLow:int=0)
   > n15 = Branch(Condition="integer >= integer"(A=@n13, B=@n14))
   Set HasError(execute=@n15.else, HasError=true)
   > Add(
    TargetArray:Array of Texts=Get Messages()
    NewItem:"Text (by ref)"=Format Text!(Format="LOCTEXT(\"Minimum LOD set for Low Platform too low\\r\\nOn mesh: {fm}  Recommended minimum: {rm}\")", fm:int=@n13, rm:int=@n14) -> (Result)
   ) {is=title K2Node_CallArrayFunction}
   Add(execute=@n12.else, TargetArray=Get Messages(), NewItem=Format Text!(Format="LOCTEXT(\"No Min LOD set for Low Platform\")") -> (Result))
   > Set HasError(HasError=true)
   n16 = Find(TargetMap=@n10, Key=Medium)
   n17 = Branch(execute=@n3.then_1, Condition=@n16.ReturnValue)
   > n18 = Set FoundMinLodForMedium(FoundMinLodForMedium:int=@n16) -> (Output_Get:int=0)
   n19 = Get MinLOD_ForMedium() -> (MinLOD_ForMedium:int=0)
   > n20 = Branch(Condition="integer >= integer"(A=@n18, B=@n19))
   Set HasError(execute=@n20.else, HasError=true)
   > Add(
    TargetArray=Get Messages()
    NewItem=Format Text!(Format="LOCTEXT(\"Minimum LOD set for Medium Platform too low\\r\\nOn mesh: {fm}  Recommended minimum: {rm}\")", fm:int=@n18, rm:int=@n19) -> (Result)
   )
   Add(execute=@n17.else, TargetArray=Get Messages(), NewItem=Format Text!(Format="LOCTEXT(\"No Min LOD set for Medium Platform\")") -> (Result))
   > Set HasError(HasError=true)
   Return Node(execute=@n2.else, Updated Array=Get MinLODforLowArray())
   MakeLiteralString(Value=MeshFix_SetMinLODforPlatform)
  ]}
 MeshCheck_NaniteDisabled {nodes [
   n1 = MeshCheck_NaniteDisabled!() -> (
    then
    Static Mesh:Static Mesh Object Reference
    NaniteNotEnabledArray:Map of Static Mesh Object References to S Asset Errors Asset Health Structures
    Only Major:bool=~
   ) {is=title K2Node_FunctionEntry}
   > n2 = Branch(Condition=IsMajorIssue_Mesh(StaticMesh=@n1, Only Major=@n1.Only Major))
   n3 = Get StaticMeshEditorSubsystem()
   > n4 = Branch(Condition=Break Mesh Nanite Settings!(MeshNaniteSettings:"Mesh Nanite Settings Structure (by ref)"=GetNaniteSettings!(self:Static Mesh Editor Subsystem Object Reference=@n3, StaticMesh:Static Mesh Object Reference=@n1) -> (ReturnValue:Mesh Nanite Settings Structure)) -> (
     bEnabled:bool=~
     bExplicitTangents:bool=~
     bLerpUVs:bool=~
     bSeparable:bool=~
     bVoxelNDF:bool=~
     bVoxelOpacity:bool=~
     ShapePreservation:ENaniteShapePreservation Enum
     PositionPrecision:int
     NormalPrecision:int
     TangentPrecision:int
     BoneWeightPrecision:int
     KeepPercentTriangles:float=~
     TrimRelativeError:float=~
     GenerateFallback:ENaniteGenerateFallback Enum
     FallbackTarget:ENaniteFallbackTarget Enum
     FallbackPercentTriangles:float=~
     FallbackRelativeError:float=~
     MaxEdgeLengthFactor:float=~
     NumRays:int
     VoxelLevel:int
     RayBackUp:float=~
     DisplacementUVChannel:int
    ) {is=title K2Node_BreakStruct})
   > Return Node(Updated Array=Get NaniteNotEnabledArray() -> (NaniteNotEnabledArray:Map of Static Mesh Object References to S Asset Errors Asset Health Structures))
   n5 = Get Mesh Check Fix Tooltips(execute=@n4.else)
   > n6 = Update StaticMesh Error Array(
    The Static Mesh=Get Static Mesh()
    The Array=Get NaniteNotEnabledArray()
    The Error=Format Text(Format="LOCTEXT(\"Nanite Not Enabled\")")
    Tooltip=Find(TargetMap:Map of Strings to Texts=@n5, Key:"String (by ref)"=MeshCheck_NaniteDisabled) -> (Value:Text)
   )
   > Return Node(Updated Array=@n6)
   Return Node(execute=@n2.else, Updated Array=Get NaniteNotEnabledArray())
   MakeLiteralString(Value=MeshFix_TurnOnNanite)
  ]}
 MeshCheck_NumMatSlots {nodes [
   n1 = MeshCheck_NumMatSlots!() -> (
    then
    Static Mesh:Static Mesh Object Reference
    Acceptable Num Mat Slots:int=1
    LOD To Check:int
    NumMatSlotsArray:Map of Static Mesh Object References to S Asset Errors Asset Health Structures
    Only Major:bool=~
   ) {is=title K2Node_FunctionEntry}
   > n2 = Branch(Condition=IsMajorIssue_Mesh(StaticMesh=@n1, Only Major=@n1.Only Major))
   > n3 = Set NumSections(NumSections:int=GetNumSections(self=@n1, InLOD=@n1.LOD To Check) -> (ReturnValue=0)) -> (Output_Get:int=0)
   > n4 = Branch(Condition="integer <= integer"(A=@n3, B=Get Acceptable Num Mat Slots() -> (Acceptable Num Mat Slots:int=0)))
   > Return Node(Updated Array=Get NumMatSlotsArray() -> (NumMatSlotsArray:Map of Static Mesh Object References to S Asset Errors Asset Health Structures))
   n5 = Get Mesh Check Fix Tooltips(execute=@n4.else)
   n6 = Get Acceptable Num Mat Slots() -> (Acceptable Num Mat Slots=0)
   n7 = Get NumSections() -> (NumSections:int=0)
   > n8 = Update StaticMesh Error Array(
    The Static Mesh=Get Static Mesh()
    The Array=Get NumMatSlotsArray()
    The Error="LOCTEXT(\"Too Many Materials\")"
    Tooltip=Format Text!(
     Format="LOCTEXT(\"{0}\\r\\n\\r\\n{1}\")"
     0:Text=Format Text(Format="LOCTEXT(\"Has: {0} Acceptable: {1}\")", 0:int=@n7, 1:int=@n6)
     1:Text=Find(TargetMap=@n5, Key=MeshCheck_NumMatSlots)
    ) -> (Result)
   )
   > Return Node(Updated Array=@n8)
   Return Node(execute=@n2.else, Updated Array=Get NumMatSlotsArray())
  ]}
 MeshCheck_SimilarLODs {nodes [
   n1 = MeshCheck_SimilarLODs!() -> (
    then
    Static Mesh:Static Mesh Object Reference
    AcceptablePercent:int=25
    SimilarLODS_Array:Map of Static Mesh Object References to S Asset Errors Asset Health Structures
    Only Major:bool=~
   ) {is=title K2Node_FunctionEntry}
   > n2 = Branch(Condition=IsMajorIssue_Mesh(StaticMesh=@n1, Only Major=@n1.Only Major))
   > n3 = Branch(Condition=Get Only Major() -> (Only Major:bool))
   n4 = "To Float (Integer)"(InInt=Get AcceptablePercent() -> (AcceptablePercent:int=0))
   > n5 = Set Adjusted Acceptable Percent(Adjusted Acceptable Percent:int=Round(A=float * float(A=@n4, B=1.1))) -> (Output_Get:int=0)
   n6 = Sequence!(execute=[@n23.then, @n5.then]) -> (then_0, then_1)
   n7 = Get Static Mesh()
   > n8 = For Loop(LastIndex=int - int(A=GetNumLODs(self=@n7), B=1))
   > Add(
    TargetArray:Array of Integers=Get l_lodCounts() -> (l_lodCounts:Array of Integers)
    NewItem:"Integer (by ref)"=GetNumberVerts(self=Get StaticMeshEditorSubsystem(), StaticMesh=Get Static Mesh(), LODIndex=@n8) -> (ReturnValue=0)
   )
   n9 = Get l_lodCounts()
   n10 = Branch(
    execute=@n6.then_1
    Condition="integer > integer"(A=Length(TargetArray:Array of Integers=@n9) {is=title K2Node_CallArrayFunction}, B=1) {is=title K2Node_CallFunction}
   )
   > n11 = For Loop(FirstIndex=1, LastIndex=Last Index!(TargetArray:Array of Integers=Get l_lodCounts()) -> (ReturnValue=0))
   Return Node(
    execute=@n11.Completed
    Updated Array=Get SimilarLODS_Array() -> (SimilarLODS_Array:Map of Static Mesh Object References to S Asset Errors Asset Health Structures)
   )
   n12 = Get l_lodCounts()
   n13 = "Get (a copy)"(Array:Array of Integers=@n12, Dimension 1=@n11) -> (Output:int)
   n14 = "Get (a copy)"(Array=@n12, Dimension 1=int - int(A=@n11) {is=title K2Node_PromotableOperator})
   n15 = "Max (Integer)"(A=@n14, B=@n13)
   n16 = "Min (Integer)"(A=@n14, B=@n13) -> (ReturnValue=0)
   > Set l_percent(l_percent:double=float / float(A="To Float (Integer)"(InInt=@n16), B="To Float (Integer)"(InInt=@n15)) {is=title K2Node_CallFunction}) -> (Output_Get:double)
   n17 = "To Float (Integer)"(InInt=Get Adjusted Acceptable Percent() -> (Adjusted Acceptable Percent:int=0))
   n18 = Get l_percent() -> (l_percent:double)
   > Branch(Condition="float > float"(A=float / float(A=@n17, B=100), B=float - float(B=@n18) {is=title K2Node_CallFunction}) {is=title K2Node_CallFunction})
   > n19 = Get Mesh Check Fix Tooltips()
   n20 = Round(A=float * float(A=Get l_percent(), B=100))
   n21 = Get AcceptablePercent() -> (AcceptablePercent=0)
   > n22 = Update StaticMesh Error Array(
    The Static Mesh=Get Static Mesh()
    The Array=Get SimilarLODS_Array()
    The Error="LOCTEXT(\"LODs Too Similar\")"
    Tooltip=Format Text(
     Format="LOCTEXT(\"{0}\\r\\n\\r\\n{1}\")"
     0:Text=Format Text!(Format="LOCTEXT(\"Triangle reduction is only {a}%\\r\\nIdeal reduction is {b}% or less\")", a:int=@n20, b:int=@n21) -> (Result)
     1:Text=Find(TargetMap=@n19, Key=MeshCheck_SimilarLODs)
    )
   )
   > Return Node(Updated Array=@n22)
   Return Node(execute=@n10.else, Updated Array=Get SimilarLODS_Array())
   n23 = Set Adjusted Acceptable Percent(execute=@n3.else, Adjusted Acceptable Percent=Get AcceptablePercent() -> (AcceptablePercent=0)) -> (Output_Get=0)
   Return Node(execute=@n2.else, Updated Array=Get SimilarLODS_Array())
  ]}
 MeshCheck_TooManyVerts {nodes [
   n1 = MeshCheck_TooManyVerts!() -> (
    then
    Static Mesh:Static Mesh Object Reference
    LOD:int
    TooManyVertsArray:Map of Static Mesh Object References to S Asset Errors Asset Health Structures
   ) {is=title K2Node_FunctionEntry}
   n2 = Make Literal Float(Value=193)
   n3 = Make Literal Float(Value=96)
   n4 = Make Literal Float(Value=192)
   > n5 = Set RangeMap(RangeMap:"Map of Float (double-precision)s to Integers"=Make Map(
     Key 0=@n3
     Value 0=MakeLiteralInt(Value=1000) -> (ReturnValue=0)
     Key 1=@n4
     Value 1=MakeLiteralInt(Value=3000) -> (ReturnValue=0)
     Key 2=@n2
     Value 2=MakeLiteralInt(Value=5000) -> (ReturnValue=0)
    )) -> (Output_Get:"Map of Float (double-precision)s to Integers")
   > Keys!(execute, TargetMap:"Map of Float (double-precision)s to Integers"=@n5) -> (then, Keys:"Array of Float (double-precision)s")
   > Set Range Map Keys(Range Map Keys:"Array of Float (double-precision)s"=Make Array!("[0]":double=@n3, "[1]":double=@n4, "[2]":double=@n2) -> (Array:"Array of Float (double-precision)s")) -> (Output_Get:"Array of Float (double-precision)s")
   > n6 = Sequence()
   n7 = GetNumberVerts(self=Get StaticMeshEditorSubsystem(), StaticMesh=Get Static Mesh(), LODIndex=Get LOD() -> (LOD:int=0)) -> (ReturnValue=0)
   n8 = Find(
    TargetMap:"Map of Float (double-precision)s to Integers"=Get RangeMap() -> (RangeMap:"Map of Float (double-precision)s to Integers")
    Key:"Float (double-precision) (by ref)"=Get ThisObjectVertThreshold() -> (ThisObjectVertThreshold:double)
   ) -> (Value:int)
   n9 = Branch(execute=@n6.then_1, Condition="integer > integer"(A=@n7, B=@n8) {is=title K2Node_PromotableOperator})
   > n10 = Get Mesh Check Fix Tooltips()
   n11 = Get LOD() -> (LOD=0)
   n12 = "Get Size (Box)"(InBox=GetBoundingBox(self=@n1))
   n13 = BreakVector(InVec=@n12)
   n14 = "To String (Text)"(InText=Format Text!(Format="LOCTEXT(\"{x}x{y}x{z}\")", x:int=Round(A=@n13), y:int=Round(A=@n13.Y), z:int=Round(A=@n13.Z)) -> (Result))
   > n15 = Update StaticMesh Error Array(
    The Static Mesh=Get Static Mesh()
    The Array=Get TooManyVertsArray() -> (TooManyVertsArray:Map of Static Mesh Object References to S Asset Errors Asset Health Structures)
    The Error="LOCTEXT(\"Too Many Vertices\")"
    Tooltip=Format Text!(
     Format="LOCTEXT(\"{msg}\\r\\n{fixmsg}\")"
     msg:Text=Find(TargetMap:Map of Strings to Texts=@n10, Key:"String (by ref)"=MeshCheck_TooManyVerts) -> (Value:Text)
     fixmsg:Text=Format Text!(
      Format="LOCTEXT(\"LOD{0} has too many vertices {1} for object of size: {2}\\r\\n\\r\\nMaximum vertex count for this size should be: {3}\")"
      0:int=@n11
      1:int=@n7
      2:String=@n14
      3:int=@n8
     ) -> (Result)
    ) -> (Result)
   )
   > Return Node(Updated Array=@n15)
   Return Node(execute=@n9.else, Updated Array=Get TooManyVertsArray())
   Set FoundThreshold(execute=@n6.then_0, FoundThreshold:bool=false) -> (Output_Get:bool)
   > n16 = For Each Loop with Break!(
    Exec:Exec
    Array:"Array of Float (double-precision)s"=Get Range Map Keys() -> (Range Map Keys:"Array of Float (double-precision)s")
    Break:Exec=@n20.then
   ) -> (LoopBody:Exec, Array Element:double, Array Index:int, Completed:Exec)
   n17 = Branch(execute=@n16.Completed, Condition=Get FoundThreshold() -> (FoundThreshold:bool))
   n18 = Get Range Map Keys()
   n19 = Length(TargetArray:"Array of Float (double-precision)s"=@n18)
   Set ThisObjectVertThreshold(
    execute=@n17.else
    ThisObjectVertThreshold:double="Get (a ref)"(Array:"Array of Float (double-precision)s"=@n18, Dimension 1=int - int(A=@n19)) -> (Output:"Float (double-precision) (by ref)")
   ) -> (Output_Get:double)
   Branch(execute=@n16.LoopBody, Condition="float < float"(A=GetMaxElement(A=@n12), B=@n16))
   > Set ThisObjectVertThreshold(ThisObjectVertThreshold=@n16)
   > n20 = Set FoundThreshold(FoundThreshold=true)
  ]}
 TextureCheck_Compression_GrayscaleSRGB {nodes [
   n1 = TextureCheck_Compression_GrayscaleSRGB!() -> (
    then
    Texture:Texture 2D Object Reference
    Texture Array:Map of Texture 2D Object References to S Asset Errors Asset Health Structures
    Only Major:bool=~
    Major Texture Threshold:int=1024
   ) {is=title K2Node_FunctionEntry}
   > n2 = Branch(Condition=IsMajorIssue_Texture!(
     Texture:Texture 2D Object Reference=@n1
     Only Major:bool=@n1.Only Major
     Major Texture Threshold:int=@n1.Major Texture Threshold
     __WorldContext:Object Reference
    ) -> (Continue Test:bool) {is=title K2Node_CallFunction})
   n3 = Get Texture() -> (Texture:Texture 2D Object Reference)
   n4 = Get CompressionSettings(self=@n3)
   > n5 = Branch(Condition=AND Boolean(
     A="Equal (Enum)"(A:TextureCompressionSettings Enum=@n4, B:TextureCompressionSettings Enum=TC_Grayscale)
     B=Get SRGB!(self:Texture Object Reference=@n3) -> (SRGB:bool)
    ))
   > n6 = Get Texture Check Fix Tooltips!(execute, __WorldContext:Object Reference) -> (then, MeshCheck_NumMatSlots:Map of Strings to Texts) {is=title K2Node_CallFunction}
   > n7 = Update Texture Error Array!(
    execute
    The Texture:Texture 2D Object Reference=Get Texture()
    The Array:Map of Texture 2D Object References to S Asset Errors Asset Health Structures=Get Texture Array() -> (Texture Array:Map of Texture 2D Object References to S Asset Errors Asset Health Structures)
    The Error:Text="LOCTEXT(\"Grayscale sRGB\")"
    The Fix If Any:String
    Tooltip:Text=Find(TargetMap=@n6, Key=TextureCheck_Compression_GrayscaleSRGB)
    __WorldContext:Object Reference
   ) -> (then, Updated Array:Map of Texture 2D Object References to S Asset Errors Asset Health Structures)
   > Return Node!(execute, Is Ok Grayscale:bool=false, Updated Texture Array:Map of Texture 2D Object References to S Asset Errors Asset Health Structures=@n7)
   Return Node(execute=@n5.else, Is Ok Grayscale=true, Updated Texture Array=Get Texture Array())
   Return Node(execute=@n2.else, Is Ok Grayscale=false, Updated Texture Array=Get Texture Array())
  ]}
 TextureCheck_Compression_Uncompressed {nodes [
   n1 = TextureCheck_Compression_Uncompressed!() -> (
    then
    Texture:Texture 2D Object Reference
    Texture Array:Map of Texture 2D Object References to S Asset Errors Asset Health Structures
    Only Major:bool=~
    Major Texture Threshold:int=1024
   ) {is=title K2Node_FunctionEntry}
   > n2 = Branch(Condition=IsMajorIssue_Texture(Texture=@n1, Only Major=@n1.Only Major, Major Texture Threshold=@n1.Major Texture Threshold))
   n3 = Get Texture()
   > n4 = Branch(Condition="Equal (Enum)"(A=Get CompressionSettings(self=@n3), B=TC_EditorIcon))
   > n5 = Get Texture Check Fix Tooltips()
   > n6 = Update Texture Error Array(
    The Texture=Get Texture()
    The Array=Get Texture Array()
    The Error="LOCTEXT(\"Texture Uncompressed\")"
    Tooltip=Find(TargetMap=@n5, Key=TextureCheck_Compression_Uncompressed)
   )
   > Return Node!(execute, Is Ok Compression:bool=false, Updated Texture Array=@n6)
   Return Node(execute=@n4.else, Is Ok Compression=true, Updated Texture Array=Get Texture Array())
   Return Node!(execute=@n2.else, Updated Texture Array=Get Texture Array(), Is Ok Compression=false)
  ]}
 TextureCheck_Compression_VectorDisplacement {nodes [
   n1 = TextureCheck_Compression_VectorDisplacement!() -> (
    then
    Texture:Texture 2D Object Reference
    Texture Array:Map of Texture 2D Object References to S Asset Errors Asset Health Structures
    Only Major:bool=~
    Major Texture Threshold:int=1024
   ) {is=title K2Node_FunctionEntry}
   > n2 = Branch(Condition=IsMajorIssue_Texture(Texture=@n1, Only Major=@n1.Only Major, Major Texture Threshold=@n1.Major Texture Threshold))
   n3 = Get Texture()
   > n4 = Branch(Condition="Equal (Enum)"(A=Get CompressionSettings(self=@n3), B=TC_VectorDisplacementmap))
   > n5 = Get Texture Check Fix Tooltips()
   > n6 = Update Texture Error Array(
    The Texture=Get Texture()
    The Array=Get Texture Array()
    The Error="LOCTEXT(\"Vector Displacement\")"
    Tooltip=Find(TargetMap=@n5, Key=TextureCheck_Compression_VectorDisplacement)
   )
   > Return Node!(execute, Is Ok Compression=false, Updated Texture Array=@n6)
   Return Node(execute=@n4.else, Is Ok Compression=true, Updated Texture Array=Get Texture Array())
   Return Node!(execute=@n2.else, Updated Texture Array=Get Texture Array(), Is Ok Compression=false)
   MakeLiteralString(Value=TextureFix_Compression_SetToBC7)
  ]}
 TextureCheck_IsTooLarge {nodes [
   n1 = TextureCheck_IsTooLarge!() -> (
    then
    Texture:Texture 2D Object Reference
    Size Threshold:int=2048
    Too Large Array:Map of Texture 2D Object References to S Asset Errors Asset Health Structures
   ) {is=title K2Node_FunctionEntry}
   n2 = GetSizeY(self=@n1) -> (ReturnValue=0)
   n3 = GetSizeX(self=@n1) -> (ReturnValue=0)
   > n4 = Branch(Condition=OR Boolean(A="integer > integer"(A=@n3, B=@n1.Size Threshold), B="integer > integer"(A=@n2, B=@n1.Size Threshold)))
   > n5 = Get Texture Check Fix Tooltips()
   > n6 = Update Texture Error Array(
    The Texture=Get Texture()
    The Array=Get Too Large Array() -> (Too Large Array:Map of Texture 2D Object References to S Asset Errors Asset Health Structures)
    The Error="LOCTEXT(\"Texture Too Large\")"
    Tooltip=Format Text!(Format="LOCTEXT(\"Texture is {x} x {y}. \\r\\n\\r\\n{tt}\")", x:int=@n3, y:int=@n2, tt:Text=Find(TargetMap=@n5, Key=TextureCheck_IsTooLarge)) -> (Result)
   )
   > Return Node!(execute, Updated Array:Map of Texture 2D Object References to S Asset Errors Asset Health Structures=@n6)
   Return Node(execute=@n4.else, Updated Array=Get Too Large Array())
   Format Text!(Format=Get Error Text() -> (Error Text:Text), 0:int=@n1.Size Threshold) -> (Result)
  ]}
 TextureCheck_NotPow2 {nodes [
   n1 = TextureCheck_NotPow2!() -> (
    then
    Texture:Texture 2D Object Reference
    Not Pow 2 Array:Map of Texture 2D Object References to S Asset Errors Asset Health Structures
    Only Major:bool=~
    Major Texture Threshold:int=1024
   ) {is=title K2Node_FunctionEntry}
   > n2 = Branch(Condition=IsMajorIssue_Texture(Texture=@n1, Only Major=@n1.Only Major, Major Texture Threshold=@n1.Major Texture Threshold))
   n3 = Get Texture()
   n4 = IsPow2!(Input Value:int=GetSizeY(self=@n3) -> (ReturnValue=0), __WorldContext:Object Reference) -> (Is Pow 2:bool) {is=title K2Node_CallFunction}
   n5 = IsPow2(Input Value=GetSizeX(self=@n3) -> (ReturnValue=0))
   > n6 = Branch(Condition=OR Boolean(A=NOT Boolean(A=@n5), B=NOT Boolean(A=@n4)))
   > n7 = Get Texture Check Fix Tooltips()
   > n8 = Update Texture Error Array(
    The Texture=Get Texture()
    The Array=Get Not Pow 2 Array() -> (Not Pow 2 Array:Map of Texture 2D Object References to S Asset Errors Asset Health Structures)
    The Error="LOCTEXT(\"Not Power of 2\")"
    Tooltip=Find(TargetMap=@n7, Key=TextureCheck_NotPow2)
   )
   > Return Node!(execute, Updated Not Pow 2 Array:Map of Texture 2D Object References to S Asset Errors Asset Health Structures=@n8)
   Return Node(execute=@n6.else, Updated Not Pow 2 Array=Get Not Pow 2 Array())
   Return Node(execute=@n2.else, Updated Not Pow 2 Array=Get Not Pow 2 Array())
  ]}
 TextureCheck_NotStreaming {nodes [
   n1 = GetEditorWorld(execute=@n9.then, self=Get UnrealEditorSubsystem())
   n2 = FortShippingLog!(execute, WorldContextObject:Object Reference=@n1, InString:String=Before Texture Group, bLogAsWarning:bool=true) -> (then)
   n3 = Get Texture()
   Branch(
    execute=[@n2.then, @n12.then_2]
    Condition="Equal (Enum)"(A:TextureGroup Enum=Get LODGroup!(self:Texture Object Reference=@n3) -> (LODGroup:TextureGroup Enum), B:TextureGroup Enum=TEXTUREGROUP_UI)
   )
   > n4 = Get Texture Check Fix Tooltips()
   > n5 = Update Texture Error Array(
    The Texture=Get Texture()
    The Array=Get Temp Not Streaming Array() -> (Temp Not Streaming Array:Map of Texture 2D Object References to S Asset Errors Asset Health Structures)
    The Error=Get Error Text Texture Group() -> (Error Text Texture Group:Text)
    Tooltip=Find(TargetMap=@n4, Key=TextureCheck_NotStreaming_GroupUI)
   )
   > Set Temp Not Streaming Array(Temp Not Streaming Array:Map of Texture 2D Object References to S Asset Errors Asset Health Structures=@n5) -> (Output_Get:Map of Texture 2D Object References to S Asset Errors Asset Health Structures)
   n6 = Get Texture()
   n7 = IsObjectCooked!(Object:Object Reference=@n6) -> (ReturnValue:bool)
   n8 = IsAssetCooked!(InAssetData:"Asset Data Structure (by ref)"=CreateAssetData(InAsset=@n6, bAllowBlueprintClass=false)) -> (ReturnValue:bool)
   n9 = FortShippingLog(
    WorldContextObject=@n1
    InString="To String (Text)"(InText=Format Text!(
      Format="LOCTEXT(\":\\r\\nTexture: {t}\\r\\nObj cooked:  {o}\\r\\nAsset Cook: {a}\")"
      t:Texture 2D Object Reference=@n6
      o:bool=@n7
      a:bool=@n8
     ) -> (Result))
    bLogAsWarning=true
   )
   n10 = TextureCheck_NotStreaming!() -> (
    then
    Texture:Texture 2D Object Reference
    Not Streaming Array:Map of Texture 2D Object References to S Asset Errors Asset Health Structures
    Only Major:bool=~
    Major Texture Threshold:int=1024
   ) {is=title K2Node_FunctionEntry}
   n11 = FortShippingLog(WorldContextObject=@n1, InString=GetSoftObjectPath!(Object=@n10) -> (ReturnValue_PathString:String), bLogAsWarning=true)
   Set Temp Not Streaming Array(execute=[@n1.then, @n11.then, @n20.then], Temp Not Streaming Array=@n10.Not Streaming Array)
   > n12 = Sequence!(execute) -> (then_0, then_1, then_2:Exec, then_3:Exec)
   Return Node!(
    execute=@n12.then_3
    Updated Temp Not Streaming Array:Map of Texture 2D Object References to S Asset Errors Asset Health Structures=Get Temp Not Streaming Array()
   )
   n13 = Get Texture()
   Branch(
    execute=[@n18.then, @n12.then_1]
    Condition="Equal (Enum)"(
     A:TextureMipGenSettings Enum=Get MipGenSettings!(self:Texture Object Reference=@n13) -> (MipGenSettings:TextureMipGenSettings Enum=TMGS_FromTextureGroup)
     B:TextureMipGenSettings Enum=TMGS_NoMipmaps
    )
   )
   > n14 = Get Texture Check Fix Tooltips()
   > n15 = Update Texture Error Array(
    The Texture=Get Texture()
    The Array=Get Temp Not Streaming Array()
    The Error="LOCTEXT(\"No MipMaps\")"
    Tooltip=Find(TargetMap=@n14, Key=TextureCheck_NotStreaming_NoMips)
   )
   > Set Temp Not Streaming Array(Temp Not Streaming Array=@n15)
   FortShippingLog(WorldContextObject=@n1, InString=Before Get Texture Streaming Method, bLogAsWarning=true)
   n16 = Get Texture()
   > Branch(Condition="Equal (Byte)"(A:Byte=GetTextureStreamingMethod!(self:Texture Object Reference=@n16) -> (ReturnValue:ETextureStreamingMethod Enum=TSM_Invalid), B:Byte=1) -> (ReturnValue:bool=~) {is=title K2Node_PromotableOperator})
   > n17 = Update Texture Error Array(
    The Texture=Get Texture()
    The Array=Get Temp Not Streaming Array()
    The Error=Get Error Text Never Stream() -> (Error Text Never Stream:Text)
   )
   > Set Temp Not Streaming Array(Temp Not Streaming Array=@n17)
   n18 = FortShippingLog(WorldContextObject=@n1, InString=Before Mip Get Settings, bLogAsWarning=true)
   Return Node(execute=@n20.else, Updated Temp Not Streaming Array=@n10.Not Streaming Array)
   Get NeverStream!(self:Streamable Render Asset Object Reference=@n16) -> (NeverStream:bool)
   n19 = IsMajorIssue_Texture(Texture=@n10, Only Major=@n10.Only Major, Major Texture Threshold=@n10.Major Texture Threshold)
   n20 = Branch(execute=@n10.then, Condition=@n19)
  ]}
 Update SoundWave Error Array {nodes [
   n1 = Update SoundWave Error Array!() -> (
    then
    The Sound Wave:Sound Wave Object Reference
    The Array:Map of Sound Wave Object References to S Asset Errors Asset Health Structures
    The Error:Text
   ) {is=title K2Node_FunctionEntry}
   > n2 = Set Temp Local Array(Temp Local Array:Map of Sound Wave Object References to S Asset Errors Asset Health Structures=@n1.The Array) -> (Output_Get:Map of Sound Wave Object References to S Asset Errors Asset Health Structures)
   n3 = Find(TargetMap:Map of Sound Wave Object References to S Asset Errors Asset Health Structures=@n2, Key:"Sound Wave Object Reference (by ref)"=@n1) -> (Value:S Asset Errors Asset Health Structure)
   > Set Found Struct(Found Struct:S Asset Errors Asset Health Structure=@n3) -> (Output_Get:S Asset Errors Asset Health Structure)
   > n4 = Branch(Condition=@n3.ReturnValue)
   > n5 = Set Temp Error Array(Temp Error Array:Array of Texts=Break S Asset Errors Asset Health!(S_AssetErrors_AssetHealth:"S Asset Errors Asset Health Structure (by ref)"=Get Found Struct() -> (Found Struct:S Asset Errors Asset Health Structure)) -> (
     Errors_3_#1:Array of Texts
     Fixes_14_#2:Map of Strings to Strings
     ErrorTooltips_19_#3:Array of Texts
    ) {is=title K2Node_BreakStruct}) -> (Output_Get:Array of Texts)
   > Add Unique(TargetArray:Array of Texts=@n5, NewItem:"Text (by ref)"=Get The Error() -> (The Error:Text))
   > Remove!(
    execute
    TargetMap:Map of Sound Wave Object References to S Asset Errors Asset Health Structures=Get Temp Local Array() -> (Temp Local Array:Map of Sound Wave Object References to S Asset Errors Asset Health Structures)
    Key:"Sound Wave Object Reference (by ref)"=Get The Sound Wave() -> (The Sound Wave:Sound Wave Object Reference)
   ) -> (then, ReturnValue:bool)
   > Add!(
    execute
    TargetMap:Map of Sound Wave Object References to S Asset Errors Asset Health Structures=Get Temp Local Array()
    Key:"Sound Wave Object Reference (by ref)"=Get The Sound Wave()
    Value:"S Asset Errors Asset Health Structure (by ref)"=Make S Asset Errors Asset Health!(
     Errors_3_#1:Array of Texts=Get Temp Error Array() -> (Temp Error Array:Array of Texts)
     Fixes_14_#2:Map of Strings to Strings=()
     ErrorTooltips_19_#3:Array of Texts
    ) -> (S_AssetErrors_AssetHealth:S Asset Errors Asset Health Structure) {is=title K2Node_MakeStruct}
   ) -> (then) {is=title K2Node_CallFunction}
   > Return Node!(execute, Updated Array:Map of Sound Wave Object References to S Asset Errors Asset Health Structures=Get Temp Local Array())
   Add(
    execute=@n4.else
    TargetMap=Get Temp Local Array()
    Key=Get The Sound Wave()
    Value=Make S Asset Errors Asset Health(
     Errors_3_#1=Make Array!("[0]":Text=@n1.The Error) -> (Array:Array of Texts)
     Fixes_14_#2=()
    )
   )
   > Return Node(Updated Array=Get Temp Local Array())
  ]}
 Update StaticMesh Error Array {nodes [
   n1 = Update StaticMesh Error Array!() -> (
    then
    The Static Mesh:Static Mesh Object Reference
    The Array:Map of Static Mesh Object References to S Asset Errors Asset Health Structures
    The Error:Text
    The Fix if Any:String
    Tooltip:Text
   ) {is=title K2Node_FunctionEntry}
   > n2 = Set Temp Local Array(Temp Local Array:Map of Static Mesh Object References to S Asset Errors Asset Health Structures=@n1.The Array) -> (Output_Get:Map of Static Mesh Object References to S Asset Errors Asset Health Structures)
   n3 = Find(
    TargetMap:Map of Static Mesh Object References to S Asset Errors Asset Health Structures=@n2
    Key:"Static Mesh Object Reference (by ref)"=Get The Static Mesh() -> (The Static Mesh:Static Mesh Object Reference)
   )
   > Set Found Struct(Found Struct=@n3)
   > n4 = Branch(Condition=@n3.ReturnValue)
   n5 = Break S Asset Errors Asset Health(S_AssetErrors_AssetHealth=Get Found Struct())
   > n6 = Set Temp Fix Map(Temp Fix Map:Map of Strings to Strings=@n5.Fixes_14_#2) -> (Output_Get:Map of Strings to Strings)
   > n7 = Set Temp Error Array(Temp Error Array=@n5)
   n8 = Get The Error()
   > Add Unique(TargetArray=@n7, NewItem=@n8)
   > n9 = Set Temp Tooltip Array(Temp Tooltip Array:Array of Texts=@n5.ErrorTooltips_19_#3) -> (Output_Get:Array of Texts)
   > Add Unique(TargetArray=@n9, NewItem=Get Tooltip() -> (Tooltip:Text))
   > Add(
    TargetMap:Map of Strings to Strings=@n6
    Key:"String (by ref)"="To String (Text)"(InText=@n8)
    Value:"String (by ref)"=Get The Fix If Any() -> (The Fix If Any:String)
   )
   > Remove(
    TargetMap:Map of Static Mesh Object References to S Asset Errors Asset Health Structures=Get Temp Local Array() -> (Temp Local Array:Map of Static Mesh Object References to S Asset Errors Asset Health Structures)
    Key:"Static Mesh Object Reference (by ref)"=Get The Static Mesh()
   )
   > Add(
    TargetMap:Map of Static Mesh Object References to S Asset Errors Asset Health Structures=Get Temp Local Array()
    Key:"Static Mesh Object Reference (by ref)"=Get The Static Mesh()
    Value:"S Asset Errors Asset Health Structure (by ref)"=Make S Asset Errors Asset Health(
     Errors_3_#1=Get Temp Error Array()
     Fixes_14_#2=Get Temp Fix Map() -> (Temp Fix Map:Map of Strings to Strings)
     ErrorTooltips_19_#3=Get Temp Tooltip Array() -> (Temp Tooltip Array:Array of Texts)
    )
   )
   > Return Node(Updated Array:Map of Static Mesh Object References to S Asset Errors Asset Health Structures=Get Temp Local Array())
   n10 = Make Array("[0]"=@n1.The Error)
   n11 = Make Array("[0]"=@n1.Tooltip)
   n12 = "To String (Text)"(InText=@n1.The Error)
   n13 = Make Map!(Key 0:String=@n12, Value 0:String=NONE) -> (Map:Map of Strings to Strings)
   n14 = Make Map(Key 0=@n12, Value 0=@n1.The Fix if Any)
   Add(
    execute=@n4.else
    TargetMap=Get Temp Local Array()
    Key=Get The Static Mesh()
    Value=Select(
     Option 0:S Asset Errors Asset Health Structure=Make S Asset Errors Asset Health(
      Errors_3_#1=@n10
      Fixes_14_#2=@n14
      ErrorTooltips_19_#3=@n11
     )
     Option 1:S Asset Errors Asset Health Structure=Make S Asset Errors Asset Health(
      Errors_3_#1=@n10
      Fixes_14_#2=@n13
      ErrorTooltips_19_#3=@n11
     )
     Index=IsEmpty(InString=@n1.The Fix if Any)
    ) -> (ReturnValue:S Asset Errors Asset Health Structure)
   )
   > Return Node(Updated Array=Get Temp Local Array())
  ]}
 Update Texture Error Array {nodes [
   n1 = Update Texture Error Array!() -> (
    then
    The Texture:Texture 2D Object Reference
    The Array:Map of Texture 2D Object References to S Asset Errors Asset Health Structures
    The Error:Text
    The Fix If Any:String
    Tooltip:Text
   ) {is=title K2Node_FunctionEntry}
   > n2 = Set Temp Local Array(Temp Local Array:Map of Texture 2D Object References to S Asset Errors Asset Health Structures=@n1.The Array) -> (Output_Get:Map of Texture 2D Object References to S Asset Errors Asset Health Structures)
   n3 = Find(
    TargetMap:Map of Texture 2D Object References to S Asset Errors Asset Health Structures=@n2
    Key:"Texture 2D Object Reference (by ref)"=Get The Texture() -> (The Texture:Texture 2D Object Reference)
   )
   > Set Found Struct(Found Struct=@n3)
   > n4 = Branch(Condition=@n3.ReturnValue)
   n5 = Break S Asset Errors Asset Health(S_AssetErrors_AssetHealth=Get Found Struct())
   > n6 = Set Temp Fix Map(Temp Fix Map=@n5.Fixes_14_#2)
   > n7 = Set Temp Error Array(Temp Error Array=@n5)
   n8 = Get The Error()
   > n9 = Add Unique(TargetArray=@n7, NewItem=@n8)
   Add(
    execute=[@n16.then, @n9.then]
    TargetMap:Map of Strings to Strings=@n6
    Key:"String (by ref)"="To String (Text)"(InText=@n8)
    Value:"String (by ref)"=Get The Fix If Any()
   )
   > Remove(
    TargetMap:Map of Texture 2D Object References to S Asset Errors Asset Health Structures=Get Temp Local Array() -> (Temp Local Array:Map of Texture 2D Object References to S Asset Errors Asset Health Structures)
    Key:"Texture 2D Object Reference (by ref)"=Get The Texture()
   )
   > Add(
    TargetMap:Map of Texture 2D Object References to S Asset Errors Asset Health Structures=Get Temp Local Array()
    Key:"Texture 2D Object Reference (by ref)"=Get The Texture()
    Value:"S Asset Errors Asset Health Structure (by ref)"=Make S Asset Errors Asset Health(
     Errors_3_#1=Get Temp Error Array()
     Fixes_14_#2=Get Temp Fix Map()
     ErrorTooltips_19_#3=Get Temp Tooltip Array()
    )
   )
   > Return Node(Updated Array:Map of Texture 2D Object References to S Asset Errors Asset Health Structures=Get Temp Local Array())
   n10 = Make Array("[0]"=@n1.Tooltip)
   n11 = Make Array("[0]"=@n1.The Error)
   n12 = "To String (Text)"(InText=@n1.The Error)
   n13 = Make Map(Key 0=@n12, Value 0=NONE)
   n14 = Make Map(Key 0=@n12, Value 0=@n1.The Fix If Any)
   Add(
    execute=@n4.else
    TargetMap=Get Temp Local Array()
    Key=Get The Texture()
    Value=Select(
     Option 0=Make S Asset Errors Asset Health(
      Errors_3_#1=@n11
      Fixes_14_#2=@n14
      ErrorTooltips_19_#3=@n10
     )
     Option 1=Make S Asset Errors Asset Health(
      Errors_3_#1=@n11
      Fixes_14_#2=@n13
      ErrorTooltips_19_#3=@n10
     )
     Index=IsEmpty(InString=@n1.The Fix If Any)
    )
   )
   > Return Node(Updated Array=Get Temp Local Array())
   n15 = Set Temp Tooltip Array(Temp Tooltip Array=@n5.ErrorTooltips_19_#3)
   > n16 = Add Unique(TargetArray=@n15, NewItem=Get Tooltip())
  ]}
]
variables []
