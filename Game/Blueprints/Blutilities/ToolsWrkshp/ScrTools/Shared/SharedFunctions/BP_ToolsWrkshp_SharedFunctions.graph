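# Editor utility Blueprint of shared helper functions for the Tools Workshop scriptable tools, including asset health tooling.
# Drawing helpers turn box corners into twelve-line bounding boxes and render snap grids onto a canvas and quad.
# Editor helpers manage colored parent-child collections, asset scans, asset editors, game view and selection that preserves temporary hidden states.
# Others snap vectors, convert units, clear timed warnings and rebuild health headers with fix buttons.
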EditorUtilityBlueprint /Game/Blueprints/Blutilities/ToolsWrkshp/ScrTools/Shared/SharedFunctions/BP_ToolsWrkshp_SharedFunctions
graphs [
 CloseOpenAssetEditor {nodes [
   n1 = CloseOpenAssetEditor!() -> (then, Asset:Object Reference, Close:bool=true, Open:bool=true, OpenedMethod:EAssetTypeActivationOpenedMethod Enum=Edit) {is=title K2Node_FunctionEntry}
   > n2 = Sequence!(execute) -> (then_0, then_1, then_2:Exec)
   > Branch(Condition=@n1.Close)
   > n3 = CloseAllEditorsForAsset!(
    execute
    self:Asset Editor Subsystem Object Reference=Get AssetEditorSubsystem()
    Asset:Object Reference=Get Asset() -> (Asset:Object Reference)
   ) -> (then, ReturnValue:int=0)
   > Set Default Num Editors Closed(Default Num Editors Closed:int=@n3) -> (Output_Get:int=0)
   Branch(execute=@n2.then_1, Condition=@n1.Open)
   > n4 = OpenEditorForAssets!(
    execute
    self:Asset Editor Subsystem Object Reference=Get AssetEditorSubsystem()
    Assets:Array of Object References=Make Array("[0]":Object Reference=@n1) -> (Array:Array of Object References)
    OpenedMethod:EAssetTypeActivationOpenedMethod Enum=@n1.OpenedMethod
   ) -> (then, ReturnValue:bool)
   > Set Open Success(Open Success:bool=@n4) -> (Output_Get:bool)
   Return Node!(
    execute=@n2.then_2
    Num Editors Closed:int=Get Default Num Editors Closed() -> (Default Num Editors Closed:int=0)
    Open Succes:bool=Get Open Success() -> (Open Success:bool)
   )
  ]}
 CreateCollectionMessage {nodes [
   n1 = CreateCollectionMessage!() -> (then, CurSev:E_CollectionChoices_TextureHealth Enum=NewEnumerator0, NumFound:int, CurrentMessage:Text) {is=title K2Node_FunctionEntry}
   n2 = "To String (Text)"(InText=Get CurrentMessage() -> (CurrentMessage:Text))
   > n3 = Branch(Condition="integer > integer"(A=Len(S=@n2)))
   n4 = SelectString(A=MakeLiteralString(), B=MakeLiteralString(Value=s), bPickA="Equal (Integer)"(A=@n1.NumFound, B=1))
   n5 = Enum to String!(Enumerator:E_CollectionChoices_TextureHealth Enum=@n1) -> (ReturnValue:String)
   n6 = SelectString(
    A="To String (Text)"(InText=Format Text!(Format="LOCTEXT(\"Found {num} {sev} asset{ss}, no collection created\")", num:int=@n1.NumFound, sev:String=@n5, ss:String=@n4) -> (Result))
    B="To String (Text)"(InText=Format Text(Format="LOCTEXT(\"Found {num} {sev} asset{ss}, added to collection \\\"{sev}\\\"\")", num=@n1.NumFound, sev=@n5, ss=@n4))
    bPickA="integer < integer"(A=@n1.NumFound, B=1)
   )
   > n7 = Set OutputMessage(OutputMessage:Text=Format Text!(Format="LOCTEXT(\"{p}\\r\\n{m}\")", p:Text=Get CurrentMessage(), m:String=@n6) -> (Result)) -> (Output_Get:Text)
   Return Node!(execute=[@n8.then, @n7.then], Message:Text=Get OutputMessage() -> (OutputMessage:Text))
   n8 = Set OutputMessage(execute=@n3.else, OutputMessage="To Text (String)"(InString=@n6))
  ]}
 Destroy Collection {nodes [
   n1 = Destroy Collection!() -> (
    then
    Collection Name:String
    ShareType:ECollectionScriptingShareType Enum=Local
    CollectionScriptingContainerSource:"Collection Container Source Structure (by ref)"
   ) {is=title K2Node_FunctionEntry}
   > n2 = DestroyCollection!(
    execute
    self:Collection Manager Subsystem Object Reference=Get CollectionManagerScriptingSubsystem()
    Collection:"Collection Structure (by ref)"=Make Collection!(
     Container:Name=Break Collection Container Source!(CollectionScriptingContainerSource:"Collection Container Source Structure (by ref)"=@n1.CollectionScriptingContainerSource) -> (Name:Name, Title:Text) {is=title K2Node_BreakStruct}
     Name:Name=String To Name(InString=@n1)
     ShareType:ECollectionScriptingShareType Enum=@n1.ShareType
    ) -> (CollectionScriptingRef:Collection Structure) {is=title K2Node_MakeStruct}
   ) -> (then, ReturnValue:bool)
   > Return Node!(execute, Success:bool=@n2)
  ]}
 Draw Vertical Indicators {nodes [
   n1 = Draw Vertical Indicators!() -> (
    then
    Line Set:Scriptable Tool Line Set Object Reference
    Start:Vector=~
    Height:double=7500
    Color:Color Structure
    Thickness:float=~
    DepthBias:float=~
   ) {is=title K2Node_FunctionEntry}
   > n2 = AddLine!(execute, self:Scriptable Tool Line Set Object Reference=@n1) -> (then, ReturnValue:Scriptable Tool Line Object Reference)
   > n3 = Set Line Component(Line Component:Scriptable Tool Line Object Reference=@n2) -> (Output_Get:Scriptable Tool Line Object Reference)
   > SetLineColor!(execute, self:Scriptable Tool Line Object Reference=@n3, Color:Color Structure=@n1.Color) -> (then)
   > SetLineThickness!(
    execute
    self:Scriptable Tool Line Object Reference=Get Line Component() -> (Line Component:Scriptable Tool Line Object Reference)
    Thickness:float=@n1.Thickness
   ) -> (then)
   > SetLineDepthBias!(execute, self:Scriptable Tool Line Object Reference=Get Line Component(), DepthBias:float=@n1.DepthBias) -> (then)
   n4 = Get Height() -> (Height:double)
   n5 = Get Start() -> (Start:Vector)
   > SetLineEndPoints!(
    execute
    self:Scriptable Tool Line Object Reference=Get Line Component()
    Start:Vector=@n5
    End:Vector=vector + vector(A=@n5, B=MakeVector(Y=0, Z=@n4))
   ) -> (then)
  ]}
 DrawBBox {nodes [
   n1 = DrawBBox!() -> (
    then
    BoundOrigin:Vector=~
    BoundExtent:Vector=~
    Box Lines:Array of Scriptable Tool Line Object References
    Color:Color Structure
    Thickness:float=5
    DepthBias:float=~
   ) {is=title K2Node_FunctionEntry}
   > Clear(TargetArray:Array of Vectors=Get CurrentCorners() -> (CurrentCorners:Array of Vectors))
   > UpdateBBoxAppearance!(
    execute
    Box Lines:Array of Scriptable Tool Line Object References=@n1.Box Lines
    Color:Color Structure=@n1.Color
    Thickness:float=@n1.Thickness
    DepthBias:float=@n1.DepthBias
    __WorldContext:Object Reference
   ) -> (then)
   > n2 = For Loop(LastIndex=7)
   n3 = Sequence!(execute=@n2.Completed) -> (
    then_0
    then_1
    then_2
    then_3:Exec
    then_4:Exec
    then_5:Exec
    then_6:Exec
    then_7:Exec
    then_8:Exec
    then_9:Exec
    then_10:Exec
    then_11:Exec
   )
   > SetLinBBoxEdge!(
    execute
    Line Number:int=0
    StartPointNum:int=0
    EndPointNum:int=1
    Box Lines:Array of Scriptable Tool Line Object References=Get Box Lines() -> (Box Lines:Array of Scriptable Tool Line Object References)
    Corners:Array of Vectors=Get CurrentCorners()
    __WorldContext:Object Reference
   ) -> (then)
   SetLinBBoxEdge(execute=@n3.then_1, Line Number=1, StartPointNum=1, EndPointNum=6, Box Lines=Get Box Lines(), Corners=Get CurrentCorners())
   SetLinBBoxEdge(execute=@n3.then_10, Line Number=10, StartPointNum=6, EndPointNum=7, Box Lines=Get Box Lines(), Corners=Get CurrentCorners())
   SetLinBBoxEdge(execute=@n3.then_11, Line Number=11, StartPointNum=4, EndPointNum=2, Box Lines=Get Box Lines(), Corners=Get CurrentCorners())
   SetLinBBoxEdge(execute=@n3.then_2, Line Number=2, StartPointNum=6, EndPointNum=2, Box Lines=Get Box Lines(), Corners=Get CurrentCorners())
   SetLinBBoxEdge(execute=@n3.then_3, Line Number=3, StartPointNum=2, EndPointNum=0, Box Lines=Get Box Lines(), Corners=Get CurrentCorners())
   SetLinBBoxEdge(execute=@n3.then_4, Line Number=4, StartPointNum=0, EndPointNum=3, Box Lines=Get Box Lines(), Corners=Get CurrentCorners())
   SetLinBBoxEdge(execute=@n3.then_5, Line Number=5, StartPointNum=3, EndPointNum=5, Box Lines=Get Box Lines(), Corners=Get CurrentCorners())
   SetLinBBoxEdge(execute=@n3.then_6, Line Number=6, StartPointNum=5, EndPointNum=1, Box Lines=Get Box Lines(), Corners=Get CurrentCorners())
   SetLinBBoxEdge(execute=@n3.then_7, Line Number=7, StartPointNum=5, EndPointNum=7, Box Lines=Get Box Lines(), Corners=Get CurrentCorners())
   SetLinBBoxEdge(execute=@n3.then_8, Line Number=8, StartPointNum=7, EndPointNum=4, Box Lines=Get Box Lines(), Corners=Get CurrentCorners())
   SetLinBBoxEdge(execute=@n3.then_9, Line Number=9, StartPointNum=4, EndPointNum=3, Box Lines=Get Box Lines(), Corners=Get CurrentCorners())
   n4 = vector * vector(A=Get BoundExtent() -> (BoundExtent:Vector), B=Make Literal Float(Value=1)) -> (ReturnValue=~)
   n5 = Get BoundOrigin() -> (BoundOrigin:Vector)
   n6 = Get BoundOrigin()
   Add(
    execute=@n2.LoopBody
    TargetArray:Array of Vectors=Get CurrentCorners()
    NewItem:"Vector (by ref)"=GetBoxCorner!(Box_Min:Vector=vector - vector(A=@n6, B=@n4), Box_Max:Vector=vector + vector(A=@n5, B=@n4), CornerIndex:int=@n2) -> (ReturnValue:Vector)
   ) {is=title K2Node_CallArrayFunction}
  ]}
 Get Bbox Side Points {nodes [
   n1 = Get Bbox Side Points!() -> (then, BBox Origin:Vector=~, BBox Extent:Vector=~, Side:int) {is=title K2Node_FunctionEntry}
   > Set ExpandedBox(ExpandedBox:Box Structure=MakeBoxWithOrigin!(Origin:"Vector (by ref)"=@n1, Extent:"Vector (by ref)"=@n1.BBox Extent) -> (ReturnValue:Box Structure)) -> (Output_Get:Box Structure)
   > n2 = Sequence!(execute) -> (then_0, then_1)
   n3 = Switch on Int!(execute:Exec=@n2.then_1, Selection:int=Get Side() -> (Side:int=0)) -> (Default:Exec, 0:Exec, 1:Exec, 2:Exec, 3:Exec, 4:Exec, 5:Exec)
   Return Node!(
    execute=@n3.0
    Point 0:Vector=Get MinCorner 0() -> (MinCorner 0:Vector)
    Point 1:Vector=Get Min+X 3() -> (Min+X 3:Vector)
    Point 2:Vector=Get Max-Z 4() -> (Max-Z 4:Vector)
    Point 3:Vector=Get Min+Y 2() -> (Min+Y 2:Vector)
   )
   Return Node(
    execute=@n3.1
    Point 0=Get MaxCorner 7() -> (MaxCorner 7:Vector)
    Point 1=Get Max-X 6() -> (Max-X 6:Vector)
    Point 2=Get Min+Z 1() -> (Min+Z 1:Vector)
    Point 3=Get Max-Y 5() -> (Max-Y 5:Vector)
   )
   Return Node(execute=@n3.2, Point 0=Get MinCorner 0(), Point 1=Get Min+Z 1(), Point 2=Get Max-Y 5(), Point 3=Get Min+X 3())
   Return Node(execute=@n3.3, Point 0=Get Min+Y 2(), Point 1=Get Max-X 6(), Point 2=Get MaxCorner 7(), Point 3=Get Max-Z 4())
   Return Node(execute=@n3.4, Point 0=Get MinCorner 0(), Point 1=Get Min+Z 1(), Point 2=Get Max-X 6(), Point 3=Get Min+Y 2())
   Return Node(execute=@n3.5, Point 0=Get Max-Z 4(), Point 1=Get Min+X 3(), Point 2=Get Max-Y 5(), Point 3=Get MaxCorner 7())
   Set MinCorner 0(
    execute=@n2.then_0
    MinCorner 0:Vector=GetBoxCorner!(Box:Box Structure=Get ExpandedBox() -> (ExpandedBox:Box Structure), CornerIndex=0) -> (ReturnValue)
   ) -> (Output_Get:Vector)
   > Set Min+Z 1(Min+Z 1:Vector=GetBoxCorner(Box=Get ExpandedBox(), CornerIndex=1)) -> (Output_Get:Vector)
   > Set Min+Y 2(Min+Y 2:Vector=GetBoxCorner(Box=Get ExpandedBox(), CornerIndex=2)) -> (Output_Get:Vector)
   > Set Min+X 3(Min+X 3:Vector=GetBoxCorner(Box=Get ExpandedBox(), CornerIndex=3)) -> (Output_Get:Vector)
   > Set Max-Z 4(Max-Z 4:Vector=GetBoxCorner(Box=Get ExpandedBox(), CornerIndex=4)) -> (Output_Get:Vector)
   > Set Max-Y 5(Max-Y 5:Vector=GetBoxCorner(Box=Get ExpandedBox(), CornerIndex=5)) -> (Output_Get:Vector)
   > Set Max-X 6(Max-X 6:Vector=GetBoxCorner(Box=Get ExpandedBox(), CornerIndex=6)) -> (Output_Get:Vector)
   > Set MaxCorner 7(MaxCorner 7:Vector=GetBoxCorner(Box=Get ExpandedBox(), CornerIndex=7)) -> (Output_Get:Vector)
  ]}
 Get Collection Colors {nodes [
   Get Collection Colors() {is=title K2Node_FunctionEntry}
   > Return Node!(
    execute
    Common Parent Color:Linear Color Structure=Get Parent Color() -> (Parent Color:Linear Color Structure)
    "Common  Severe Collection Color":Linear Color Structure=Get Severe Collection Color() -> (Severe Collection Color:Linear Color Structure)
    Common Moderate Collection Color:Linear Color Structure=Get Moderate Collection Color() -> (Moderate Collection Color:Linear Color Structure)
    Common Minor Collection Color:Linear Color Structure=Get Minor Collection Color() -> (Minor Collection Color:Linear Color Structure)
   )
  ]}
 GetActorsTempHidden {nodes [
   GetActorsTempHidden() {is=title K2Node_FunctionEntry}
   > n1 = GetAllLevelActors(self=Get EditorActorSubsystem())
   > n2 = For Each Loop(Array:Array of Actor Object References=@n1) -> (Array Element:Actor Object Reference)
   n3 = Get OnlyActorsTempHidden() -> (OnlyActorsTempHidden:Array of Actor Object References)
   Return Node!(
    execute=@n2.Completed
    ActorTempHiddenMap:Map of Actor Object References to Booleans=Get ActorTempHidden() -> (ActorTempHidden:Map of Actor Object References to Booleans)
    OnlyTempHidden:Array of Actor Object References=@n3
   )
   n4 = IsTemporarilyHiddenInEditor!(self:Actor Object Reference=@n2, bIncludeParent:bool=false) -> (ReturnValue:bool)
   > Add!(
    execute
    TargetMap:Map of Actor Object References to Booleans=Get ActorTempHidden()
    Key:"Actor Object Reference (by ref)"=@n2
    Value:"Boolean (by ref)"=@n4
   ) -> (then) {is=title K2Node_CallFunction}
   > Branch(Condition=@n4)
   > Add(TargetArray:Array of Actor Object References=@n3, NewItem:"Actor Object Reference (by ref)"=@n2) {is=title K2Node_CallArrayFunction}
  ]}
 GetConvertedUnits {nodes [
   GetConvertedUnits!() -> (then, Distance in cm:double=1, Distance Units:String=cm, Time in sec:double=1, Time Units:String=sec, Grid Units:double=16) {is=title K2Node_FunctionEntry}
   > n1 = Sequence!(execute) -> (then_0, then_1, then_2:Exec, then_3:Exec)
   > Set Conv Dist Final(Conv Dist Final:double=Get Distance in cm() -> (Distance in cm:double)) -> (Output_Get:double)
   n2 = Find(
    TargetMap:"Map of Strings to Float (double-precision)s"=Get DistConversionFactors() -> (DistConversionFactors:"Map of Strings to Float (double-precision)s")
    Key=Get Distance Units() -> (Distance Units:String)
   ) -> (Value:double)
   > n3 = Set FoundDistConvString(FoundDistConvString:bool=@n2.ReturnValue) -> (Output_Get:bool)
   > Branch(Condition=@n3)
   > Set DistScaleFactor(DistScaleFactor:double=@n2) -> (Output_Get:double)
   > n4 = Branch(Condition="float >= float"(A=@n2, B=1))
   Set Conv Dist Final(execute=[@n4.then, @n5.then], Conv Dist Final=SafeDivide(A=Get Distance in cm(), B=Get DistScaleFactor() -> (DistScaleFactor:double)))
   n5 = Set DistScaleFactor(execute=@n4.else, DistScaleFactor=Get Grid Units() -> (Grid Units:double))
   Set Conv Time Final(execute=@n1.then_1, Conv Time Final:double=Get Time in sec() -> (Time in sec:double)) -> (Output_Get:double)
   n6 = Find(
    TargetMap=Get TimeConversionFactors() -> (TimeConversionFactors:"Map of Strings to Float (double-precision)s")
    Key=Get Time Units() -> (Time Units:String)
   )
   > n7 = Set Found Time Conv String(Found Time Conv String:bool=@n6.ReturnValue) -> (Output_Get:bool)
   > Branch(Condition=AND Boolean(A=@n7, B=Get FoundDistConvString() -> (FoundDistConvString:bool)))
   > Set Time Scale Factor(Time Scale Factor:double=@n6) -> (Output_Get:double)
   > Set Conv Time Final(Conv Time Final=SafeDivide(A=Get Time in sec(), B=Get Time Scale Factor() -> (Time Scale Factor:double)))
   Branch(execute=@n1.then_2, Condition=AND Boolean(A=Get FoundDistConvString(), B=Get Found Time Conv String() -> (Found Time Conv String:bool)))
   > Set DistUnitsFinal(DistUnitsFinal:String=Get Distance Units()) -> (Output_Get:String)
   > Set TimeUnitsFinal(TimeUnitsFinal:String=Get Time Units()) -> (Output_Get:String)
   Return Node!(
    execute=@n1.then_3
    Converted distance:double=Get Conv Dist Final() -> (Conv Dist Final:double)
    Distance Units Final:String=Get DistUnitsFinal() -> (DistUnitsFinal:String)
    Converted time:double=Get Conv Time Final() -> (Conv Time Final:double)
    Time Units Final:String=Get TimeUnitsFinal() -> (TimeUnitsFinal:String)
   )
  ]}
 GetProjectAssetsWithFiltering {nodes [
   n1 = GetProjectAssetsWithFiltering!() -> (then, ProjPath:String, Filter:String=DO_NOT_OPTIMIZE) {is=title K2Node_FunctionEntry}
   > n2 = Sequence!(execute) -> (then_0, then_1)
   > n3 = GetAssetsByPath!(
    execute
    self:Asset Registry Interface=GetAssetRegistry()
    PackagePath:Name=String To Name(InString=@n1)
    bRecursive:bool=true
    bIncludeOnlyOnDiskAssets:bool=false
   ) -> (then, OutAssetData:Array of Asset Data Structures, ReturnValue:bool)
   > n4 = For Each Loop(Array:Array of Asset Data Structures=@n3) -> (Array Element:Asset Data Structure)
   n5 = GetAsset(InAssetData=@n4)
   > n6 = Branch(Condition=Contains(SearchIn=Get Object Path String(Object=@n5), Substring=Get Filter() -> (Filter:String), bUseCase=true))
   n7 = Cast To StaticMesh(execute=@n6.else, Object=@n5)
   > Add(
    TargetArray:Array of Static Mesh Object References=Get StaticMeshes() -> (StaticMeshes:Array of Static Mesh Object References)
    NewItem:"Static Mesh Object Reference (by ref)"=@n7
   )
   n8 = Cast To Texture2D(execute=@n7.CastFailed, Object=@n5)
   > Add(
    TargetArray:Array of Texture 2D Object References=Get Textures() -> (Textures:Array of Texture 2D Object References)
    NewItem:"Texture 2D Object Reference (by ref)"=@n8
   )
   Return Node!(
    execute=@n2.then_1
    Found Static Meshes:Array of Static Mesh Object References=Get StaticMeshes()
    Found Textures:Array of Texture 2D Object References=Get Textures()
   )
  ]}
 GetSetGameView {nodes [
   n1 = GetSetGameView!() -> (then, SetGameView:bool=~, GameViewVal:bool=~) {is=title K2Node_FunctionEntry}
   n2 = Get LevelEditorSubsystem()
   > n3 = GetActiveViewportConfigKey!(execute, self:Level Editor Subsystem Object Reference=@n2) -> (then, ReturnValue:Name=None)
   > Set ConfigKey(ConfigKey:Name=@n3) -> (Output_Get:Name=None)
   > n4 = Branch(Condition=@n1)
   > n5 = EditorSetGameView!(
    execute
    self:Level Editor Subsystem Object Reference=@n2
    bGameView:bool=@n1.GameViewVal
    ViewportConfigKey:Name=Get ConfigKey() -> (ConfigKey:Name=None)
   ) -> (then)
   n6 = EditorGetGameView!(execute=@n4.else, self:Level Editor Subsystem Object Reference=@n2, ViewportConfigKey:Name=Get ConfigKey() -> (ConfigKey=None)) -> (then, ReturnValue:bool)
   Return Node!(execute=[@n6.then, @n5.then], ViewportKey:Name=Get ConfigKey() -> (ConfigKey=None), IsGameView:bool=@n6)
  ]}
 Make BBox Lines in Lineset {nodes [
   n1 = Make BBox Lines in Lineset!() -> (then, Line Set For BBox:Scriptable Tool Line Set Object Reference) {is=title K2Node_FunctionEntry}
   > n2 = For Loop(LastIndex=11)
   Return Node!(
    execute=@n2.Completed
    New Box Lines:Array of Scriptable Tool Line Object References=Get Box Lines Temp() -> (Box Lines Temp:Array of Scriptable Tool Line Object References)
   )
   > n3 = AddLine(self=@n1)
   > Add(
    TargetArray:Array of Scriptable Tool Line Object References=Get Box Lines Temp()
    NewItem:"Scriptable Tool Line Object Reference (by ref)"=@n3
   )
  ]}
 Make Collection {nodes [
   n1 = Make Collection!() -> (
    then
    Collection Name:String
    Parent Collection Name:String
    ShareType:ECollectionScriptingShareType Enum=Local
    Container:Collection Container Source Structure
    Parent Color:Linear Color Structure=(R=0.8788,G=0,B=1,A=1)
    Collection Color:Linear Color Structure=(R=0,G=1,B=0.0448,A=1)
   ) {is=title K2Node_FunctionEntry}
   > n2 = Sequence()
   > n3 = CollectionExists!(
    execute
    self:Collection Manager Subsystem Object Reference=Get CollectionManagerScriptingSubsystem()
    Container:Collection Container Source Structure=@n1.Container
    Collection:Name=String To Name(InString=@n1.Parent Collection Name)
    ShareType:ECollectionScriptingShareType Enum=Get ShareType() -> (ShareType:ECollectionScriptingShareType Enum=Local)
   ) -> (then, ReturnValue:bool)
   > n4 = Branch(Condition=@n3)
   > n5 = GetCollectionsByName!(
    execute
    self:Collection Manager Subsystem Object Reference=Get CollectionManagerScriptingSubsystem()
    Container:Collection Container Source Structure=Get Container() -> (Container:Collection Container Source Structure)
    Collection:Name=String To Name(InString=Get Parent Collection Name() -> (Parent Collection Name:String))
   ) -> (then, OutCollections:Array of Collection Structures, ReturnValue:bool)
   > n6 = For Each Loop(Array:Array of Collection Structures=@n5) -> (Array Element:Collection Structure)
   > Branch(Condition="Equal (Enum)"(
     A:ECollectionScriptingShareType Enum=Break Collection!(CollectionScriptingRef:"Collection Structure (by ref)"=@n6) -> (Container:Name, Name:Name, ShareType:ECollectionScriptingShareType Enum) {is=title K2Node_BreakStruct}.ShareType
     B:ECollectionScriptingShareType Enum=Get ShareType() -> (ShareType=Local)
    ))
   > n7 = Set NewParentCollection(NewParentCollection:Collection Structure=@n6) -> (Output_Get:Collection Structure)
   > SetCollectionColor!(
    execute
    self:Collection Manager Subsystem Object Reference=Get CollectionManagerScriptingSubsystem()
    Collection:"Collection Structure (by ref)"=@n7
    Color:"Linear Color Structure (by ref)"=Get Parent Color()
   ) -> (then, ReturnValue:bool)
   n8 = CreateCollection!(
    execute=@n4.else
    self:Collection Manager Subsystem Object Reference=Get CollectionManagerScriptingSubsystem()
    Container:Collection Container Source Structure=Get Container()
    Collection:Name=String To Name(InString=Get Parent Collection Name())
    ShareType:ECollectionScriptingShareType Enum=Get ShareType() -> (ShareType=Local)
   ) -> (then, OutNewCollection:Collection Structure, ReturnValue:bool)
   > Set NewParentCollection(NewParentCollection=@n8)
   n9 = CreateOrEmptyCollection!(
    execute=@n2.then_1
    self:Collection Manager Subsystem Object Reference=Get CollectionManagerScriptingSubsystem()
    Container:Collection Container Source Structure=Get Container()
    Collection:Name=String To Name(InString=@n1)
    ShareType:ECollectionScriptingShareType Enum=Get ShareType() -> (ShareType=Local)
   ) -> (then, OutNewOrEmptyCollection:Collection Structure, ReturnValue:bool)
   > ReparentCollection!(
    execute
    self:Collection Manager Subsystem Object Reference=Get CollectionManagerScriptingSubsystem()
    Collection:"Collection Structure (by ref)"=@n9
    NewParentCollection:Collection Structure=Get NewParentCollection() -> (NewParentCollection:Collection Structure)
   ) -> (then, ReturnValue:bool)
   > SetCollectionColor(self=Get CollectionManagerScriptingSubsystem(), Collection=@n9, Color=Get Collection Color() -> (Collection Color:Linear Color Structure))
   > Return Node!(
    execute
    OutNewOrEmptyCollection:Collection Structure=@n9
    Success:bool=@n9.ReturnValue
    OutParentCollection:Collection Structure=Get NewParentCollection()
   )
  ]}
 Scr Tools Draw Grid {nodes [
   n1 = Scr Tools Draw Grid!() -> (
    then
    Location:Vector=~
    Canvas For Grid:Canvas Object Reference
    Scriptable Tool Quad:Scriptable Tool Quad Object Reference
    Snap Grid Size:int=512
    Desired Num Grid Lines:int=3
    Canvas Width:int
    Canvas Height:int
    RenderColor:Linear Color Structure=(R=1,G=1,B=1,A=1)
    Thickness:float=20
   ) {is=title K2Node_FunctionEntry}
   n2 = float * float!(A="To Float (Integer)"(InInt=@n1.Snap Grid Size), B:int=@n1.Desired Num Grid Lines, C:double=2) -> (ReturnValue) {is=title K2Node_PromotableOperator}
   > Set Box Size(Box Size:double=@n2) -> (Output_Get:double)
   > n3 = Set Bbox Extent(Bbox Extent:double=SafeDivide(A=@n2, B=2)) -> (Output_Get:double)
   > n4 = Get Bbox Side Points!(
    execute
    BBox Origin:Vector=Get Location() -> (Location:Vector)
    BBox Extent:Vector=MakeVector(X=@n3, Y=@n3, Z=0.01)
    Side:int=0
    __WorldContext:Object Reference
   ) -> (then, Point 0:Vector, Point 1:Vector, Point 2:Vector, Point 3:Vector) {is=title K2Node_CallFunction}
   > SetQuadPoints!(
    execute
    self:Scriptable Tool Quad Object Reference=Get Scriptable Tool Quad() -> (Scriptable Tool Quad:Scriptable Tool Quad Object Reference)
    A:Vector=@n4
    B:Vector=@n4.Point 1
    C:Vector=@n4.Point 2
    D:Vector=@n4.Point 3
   ) -> (then)
   > n5 = For Loop(LastIndex=float / float(A=Get Box Size() -> (Box Size:double), B:int=Get Snap Grid Size() -> (Snap Grid Size:int=0)) -> (ReturnValue:int))
   n6 = float * float!(A:int=Get Snap Grid Size() -> (Snap Grid Size=0), B=@n5) -> (ReturnValue)
   n7 = Get Box Size()
   n8 = Get Canvas Width() -> (Canvas Width:int=0)
   > Set CurScreenPos(CurScreenPos:double=float * float(A:double=SafeDivide(A=@n6, B=@n7), B=@n8)) -> (Output_Get:double)
   > n9 = Set ScaledThickness(ScaledThickness:double=Get Thickness() -> (Thickness:float)) -> (Output_Get:double)
   n10 = Get CurScreenPos() -> (CurScreenPos:double)
   > Draw Line!(
    execute
    self:Canvas Object Reference=Get Canvas For Grid() -> (Canvas For Grid:Canvas Object Reference)
    ScreenPositionA_X:double=0
    ScreenPositionA_Y:double=@n10
    ScreenPositionB_X:double="To Float (Integer)"(InInt=@n8)
    ScreenPositionB_Y:double=@n10
    Thickness:float=@n9
    RenderColor:Linear Color Structure=Get RenderColor() -> (RenderColor:Linear Color Structure)
   ) -> (then)
   n11 = Get CurScreenPos()
   > Draw Line(
    self=Get Canvas For Grid()
    ScreenPositionA_X=@n11
    ScreenPositionA_Y=0
    ScreenPositionB_X=@n11
    ScreenPositionB_Y="To Float (Integer)"(InInt=Get Canvas Width() -> (Canvas Width=0))
    Thickness=Get ScaledThickness() -> (ScaledThickness:double)
    RenderColor=Get RenderColor()
   )
  ]}
 Set Selected Level Actors Ignore Eyeball {nodes [
   n1 = Set Selected Level Actors Ignore Eyeball!() -> (then, Actors To Select:Array of Actor Object References) {is=title K2Node_FunctionEntry}
   > Clear(TargetArray:Array of Booleans=Get OverlapActorsIsTempHidden() -> (OverlapActorsIsTempHidden:Array of Booleans))
   > n2 = For Each Loop(Array:Array of Actor Object References=@n1) -> (Array Element:Actor Object Reference)
   n3 = Sequence(execute=@n2.Completed)
   > SetSelectedLevelActors(self=Get EditorActorSubsystem(), ActorsToSelect=@n1)
   n4 = For Each Loop(Exec=@n3.then_1, Array=@n1)
   > SetIsTemporarilyHiddenInEditor!(
    execute
    self:Actor Object Reference=@n4
    bIsHidden:bool="Get (a ref)"(Array:Array of Booleans=Get OverlapActorsIsTempHidden(), Dimension 1=@n4.Array Index) -> (Output:"Boolean (by ref)")
   ) -> (then)
   Add(
    execute=@n2.LoopBody
    TargetArray:Array of Booleans=Get OverlapActorsIsTempHidden()
    NewItem:"Boolean (by ref)"=IsTemporarilyHiddenInEditor(self=@n2, bIncludeParent=false)
   )
   > SetIsTemporarilyHiddenInEditor(self=@n2, bIsHidden=false)
  ]}
 SetLinBBoxEdge {nodes [
   n1 = SetLinBBoxEdge!() -> (
    then
    Line Number:int
    StartPointNum:int
    EndPointNum:int
    Box Lines:Array of Scriptable Tool Line Object References
    Corners:Array of Vectors
   ) {is=title K2Node_FunctionEntry}
   > SetLineEndPoints(
    self="Get (a copy)"(Array:Array of Scriptable Tool Line Object References=@n1.Box Lines, Dimension 1=@n1) -> (Output:Scriptable Tool Line Object Reference)
    Start="Get (a ref)"(Array:Array of Vectors=@n1.Corners, Dimension 1=@n1.StartPointNum) -> (Output:"Vector (by ref)")
    End="Get (a ref)"(Array=@n1.Corners, Dimension 1=@n1.EndPointNum)
   )
  ]}
 SnapVectorToGrid {nodes [
   n1 = SnapVectorToGrid!() -> (then, VectorToSnap:Vector=~, GridSize:int=512, Snap On:bool=~, Ignore Up:bool) {is=title K2Node_FunctionEntry}
   n2 = Vector_SnappedToGrid!(InVect:Vector=@n1, InGridSize:float="To Float (Integer)"(InInt=@n1.GridSize)) -> (ReturnValue:Vector)
   n3 = BreakVector(InVec=@n2)
   n4 = MakeVector(X=@n3, Y=@n3.Y, Z=BreakVector(InVec=@n1).Z)
   n5 = Get Ignore Up() -> (Ignore Up:bool)
   > Return Node!(execute, OutputVector:Vector=SelectVector(A=SelectVector(A=@n4, B=@n2, bPickA=@n5), B=@n1, bPickA=Get Snap On() -> (Snap On:bool)))
  ]}
 TimedClearUserWarning {nodes [
   n1 = TimedClearUserWarning!() -> (
    then
    ClearAfterTime:double=3
    CurTime:double=~
    LastClearTime:double=~
    Scriptable Tool Obj Ref:Scriptable Interactive Tool Object Reference
   ) {is=title K2Node_FunctionEntry}
   > Set LastClearTimeUpdated(LastClearTimeUpdated:double=@n1.LastClearTime) -> (Output_Get:double)
   > Branch(Condition="float > float"(A=@n1.CurTime, B=float + float(A=@n1, B=@n1.LastClearTime)))
   > Set LastClearTimeUpdated(LastClearTimeUpdated=Get CurTime())
   > ClearUserMessages!(execute, self:Scriptable Interactive Tool Object Reference=@n1.Scriptable Tool Obj Ref, bHelpMessage:bool=true, bWarningMessage:bool=true) -> (then)
   > Return Node!(execute, NewLastClearTime:double=Get LastClearTimeUpdated() -> (LastClearTimeUpdated:double))
  ]}
 Update Health Tool Header {nodes [
   n1 = Update Health Tool Header!() -> (
    then
    Header Widget:BP UW Shared Header Widget Object Reference
    Cur Asset:Object Reference
    Cur Asset Error Struct:S Asset Errors Asset Health Structure
   ) {is=title K2Node_FunctionEntry}
   > n2 = Sequence!(execute) -> (then_0, then_1, then_2:Exec, then_3:Exec, then_4:Exec, then_5:Exec)
   > n3 = For Each Loop(Array:Array of Widget Object References=GetAllChildren(self=Get MainContainerVertBox!(self:BP UW Shared Header Widget Object Reference=@n1) -> (MainContainerVertBox:Vertical Box Object Reference))) -> (Array Element:Widget Object Reference)
   > RemoveFromParent(self=@n3)
   n4 = Get Header Widget()
   n5 = AddExpandableBox(execute=@n2.then_1, self=@n4)
   > n6 = Set Readout Box(Readout Box:BP UW Expandable Box Object Reference=@n5) -> (Output_Get:BP UW Expandable Box Object Reference)
   > SetUpBox(self=@n6, Name=Get Readout Box Label!(self:BP UW Shared Header Widget Object Reference=@n4) -> (Readout Box Label:Text))
   n7 = SetupLineSolo!(
    execute=@n2.then_2
    self:BP UW Expandable Box Object Reference=Get Readout Box() -> (Readout Box:BP UW Expandable Box Object Reference)
    Current Line:BP UW Solo Line Object Reference=Get Readout Line Asset Name() -> (Readout Line Asset Name:BP UW Solo Line Object Reference)
    Dynamic Line:Vertical Box Object Reference
    Value:Text="To Text (Object)"!(InObj:Object Reference=@n1.Cur Asset) -> (ReturnValue:Text)
    ToolTipText:Text
    Button Visible:bool=false
    Bold:bool=true
   ) -> (then, Readout Line:BP UW Solo Line Object Reference)
   > Set Readout Line Asset Name(Readout Line Asset Name:BP UW Solo Line Object Reference=@n7) -> (Output_Get:BP UW Solo Line Object Reference)
   Return Node!(
    execute=@n2.then_5
    Resulting Buttons:Array of BP UW Shared Fix Me Button Object References=Get Buttons() -> (Buttons:Array of BP UW Shared Fix Me Button Object References)
   )
   n8 = Break S Asset Errors Asset Health!(S_AssetErrors_AssetHealth:"S Asset Errors Asset Health Structure (by ref)"=@n1.Cur Asset Error Struct) -> (
    Errors_3_#1:Array of Texts
    Fixes_14_#2:Map of Strings to Strings
    ErrorTooltips_19_#3:Array of Texts
   ) {is=title K2Node_BreakStruct}
   n9 = For Each Loop(Exec=@n2.then_4, Array:Array of Texts=@n8) -> (Array Element:Text)
   n10 = Find(TargetMap:Map of Strings to Strings=@n8.Fixes_14_#2, Key="To String (Text)"(InText=@n9)) -> (Value:String)
   n11 = IsEmpty(InString=@n10)
   n12 = "integer >= integer"(A=FindSubstring(SearchIn=@n10, Substring=NONE))
   > n13 = SetupLineTriple!(
    execute
    self:BP UW Expandable Box Object Reference=Get Readout Box()
    Current Line:BP UW Triple Line Object Reference=Get ErrorTripleLine() -> (ErrorTripleLine:BP UW Triple Line Object Reference)
    Dynamic Line:Vertical Box Object Reference
    Label:Text=@n9
    Value:Text="LOCTEXT(\"Requires fix\")"
    ToolTipText:Text="Get (a ref)"(Array:Array of Texts=@n8.ErrorTooltips_19_#3, Dimension 1=@n9.Array Index) -> (Output:"Text (by ref)")
    MoreInfo:bool=true
    FixAvaliable:bool=AND Boolean(A=NOT Boolean(A=@n12), B=NOT Boolean(A=@n11))
   ) -> (then, Readout Line:BP UW Triple Line Object Reference)
   > Add(
    TargetArray:Array of BP UW Shared Fix Me Button Object References=Get Buttons()
    NewItem:"BP UW Shared Fix Me Button Object Reference (by ref)"=Get BP_UW_Shared_FixMeButton!(self:BP UW Triple Line Object Reference=@n13) -> (BP_UW_Shared_FixMeButton:BP UW Shared Fix Me Button Object Reference)
   )
  ]}
 UpdateBBoxAppearance {nodes [
   n1 = UpdateBBoxAppearance!() -> (then, Box Lines:Array of Scriptable Tool Line Object References, Color:Color Structure, Thickness:float=~, DepthBias:float=~) {is=title K2Node_FunctionEntry}
   > n2 = For Each Loop(Array:Array of Scriptable Tool Line Object References=@n1) -> (Array Element:Scriptable Tool Line Object Reference)
   > SetLineDepthBias(self=@n2, DepthBias=@n1.DepthBias)
   > SetLineColor(self=@n2, Color=@n1.Color)
   > SetLineThickness(self=@n2, Thickness=@n1.Thickness)
  ]}
]
variables []
