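# Editor utility Blueprint that serves as a temporary library of asset health tests and fixes for static meshes and textures.
# Tests flag missing simple collision or LODs, vertex count, and costly texture compression or grayscale sRGB settings.
# Most fixes register a transaction, add collision shapes or change texture settings, then re-read the asset to report success.
# GeoScript_Test generates collision through Geometry Script, while MeshFix_AddCollision is an unwired stub returning false.
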EditorUtilityBlueprint /Game/Blueprints/Blutilities/ToolsWrkshp/ScrTools/Shared/SharedFunctions/TemporaryLibrary_TestsAndFixes
graphs [
 GeoScript_Test {nodes [
   n1 = Copy Mesh From Static Mesh!(
    execute=@n4.then_1
    FromStaticMeshAsset:Static Mesh Object Reference=Get Mesh!() -> (Mesh:Static Mesh Object Reference)
    ToDynamicMesh:Dynamic Mesh Object Reference=Get DynamicMesh() -> (DynamicMesh:Dynamic Mesh Object Reference)
    AssetOptions:Geometry Script Copy Mesh From Asset Options Structure=Get Asset Options() -> (Asset Options:Geometry Script Copy Mesh From Asset Options Structure)
    RequestedLOD:Geometry Script Mesh Read LOD Structure=Get Requested LOD() -> (Requested LOD:Geometry Script Mesh Read LOD Structure)
    bUseSectionMaterials:bool=true
   ) -> (Failure:Exec, Success:Exec, ReturnValue:Dynamic Mesh Object Reference)
   n2 = GenerateCollisionFromMesh!(
    execute=@n1.Success
    FromDynamicMesh:Dynamic Mesh Object Reference=@n1
    Options:Geometry Script Collision From Mesh Options Structure=Get Collision Options() -> (Collision Options:Geometry Script Collision From Mesh Options Structure)
   ) -> (then, ReturnValue:Geometry Script Simple Collision Structure)
   n3 = MergeSimpleCollisionShapes!(
    execute
    SimpleCollision:"Geometry Script Simple Collision Structure (by ref)"=@n2
    MergeOptions:"Geometry Script Merge Simple Collision Options Structure (by ref)"=Get Merge Collision Options() -> (Merge Collision Options:Geometry Script Merge Simple Collision Options Structure)
   ) -> (then, bHasMerged:bool, ReturnValue:Geometry Script Simple Collision Structure)
   SetSimpleCollisionOfStaticMesh!(
    execute=[@n3.then, @n2.then]
    SimpleCollision:"Geometry Script Simple Collision Structure (by ref)"=@n2
    StaticMesh:Static Mesh Object Reference=Get Mesh()
    Options:Geometry Script Set Simple Collision Options Structure
    StaticMeshCollisionOptions:Geometry Script Set Static Mesh Collision Options Structure=Get Static Mesh Collision Options() -> (Static Mesh Collision Options:Geometry Script Set Static Mesh Collision Options Structure)
   ) -> (then)
   > PrintString(InString=Success)
   GeoScript_Test!() -> (then, Mesh:Static Mesh Object Reference) {is=title K2Node_FunctionEntry}
   > n4 = Sequence!(execute) -> (then_0, then_1, then_2:Exec)
   > n5 = CreateDynamicMeshPool!(execute) -> (then, ReturnValue:Dynamic Mesh Pool Object Reference)
   > n6 = Set DynamicMeshPool(DynamicMeshPool:Dynamic Mesh Pool Object Reference=@n5) -> (Output_Get:Dynamic Mesh Pool Object Reference)
   > n7 = RequestMesh!(execute, self:Dynamic Mesh Pool Object Reference=@n6) -> (then, ReturnValue:Dynamic Mesh Object Reference)
   > Set DynamicMesh(DynamicMesh:Dynamic Mesh Object Reference=@n7) -> (Output_Get:Dynamic Mesh Object Reference)
   n8 = GetSimpleCollisionFromStaticMesh!(execute=@n4.then_2, StaticMesh:Static Mesh Object Reference=Get Mesh()) -> (then, ReturnValue:Geometry Script Simple Collision Structure)
   > n9 = GetSimpleCollisionShapeCount!(execute, SimpleCollision:"Geometry Script Simple Collision Structure (by ref)"=@n8) -> (then, ReturnValue:int=0)
   > Return Node!(execute, Fix Successful:bool="integer > integer"(A=@n9))
   PrintString(execute=@n1.Failure, InString=Failure)
  ]}
 MeshFix_AddCollision {nodes [
   n1 = GenerateCollisionFromMesh()
   > SetSimpleCollisionOfStaticMesh(SimpleCollision=@n1, StaticMesh=Get Mesh(), StaticMeshCollisionOptions=(bMarkAsCustomized=True))
   MeshFix_AddCollision!() -> (then, Mesh:Static Mesh Object Reference) {is=title K2Node_FunctionEntry}
   > n2 = Sequence()
   Return Node(execute=@n2.then_2, Fix Successful=false)
   SetSimpleCollisionOfStaticMesh(StaticMeshCollisionOptions=(bMarkAsCustomized=True))
   SetStaticMeshCollisionFromMesh!(
    execute
    FromDynamicMesh:Dynamic Mesh Object Reference
    ToStaticMeshAsset:Static Mesh Object Reference
    Options:Geometry Script Collision From Mesh Options Structure
    StaticMeshCollisionOptions:Geometry Script Set Static Mesh Collision Options Structure=(bMarkAsCustomized=True)
   ) -> (then, ReturnValue:Dynamic Mesh Object Reference)
  ]}
 MeshFix_AddSimpleCollision_Box {nodes [
   n1 = MeshFix_AddSimpleCollision_Box!() -> (then, Mesh:Static Mesh Object Reference) {is=title K2Node_FunctionEntry}
   > TransactObject(Object=@n1)
   > AddSimpleCollisions!(
    execute
    self:Static Mesh Editor Subsystem Object Reference=Get StaticMeshEditorSubsystem()
    StaticMesh:Static Mesh Object Reference=@n1
    ShapeType:EScriptCollisionShapeType Enum=Box
   ) -> (then, ReturnValue:int=0)
   > n2 = GetSimpleCollisionFromStaticMesh(StaticMesh=@n1)
   > n3 = GetSimpleCollisionShapeCount(SimpleCollision=@n2) -> (ReturnValue=0)
   > Return Node(Fix Successful="integer > integer"(A=@n3))
  ]}
 MeshFix_AddSimpleCollision_NDOP26 {nodes [
   n1 = MeshFix_AddSimpleCollision_NDOP26!() -> (then, Mesh:Static Mesh Object Reference) {is=title K2Node_FunctionEntry}
   > TransactObject(Object=@n1)
   > AddSimpleCollisions(self=Get StaticMeshEditorSubsystem(), StaticMesh=@n1, ShapeType=NDOP26) -> (ReturnValue=0)
   > n2 = GetSimpleCollisionFromStaticMesh(StaticMesh=@n1)
   > n3 = GetSimpleCollisionShapeCount(SimpleCollision=@n2) -> (ReturnValue=0)
   > Return Node(Fix Successful="integer > integer"(A=@n3))
  ]}
 MeshFix_AddSimpleCollision_Sphere {nodes [
   n1 = MeshFix_AddSimpleCollision_Sphere!() -> (then, Mesh:Static Mesh Object Reference) {is=title K2Node_FunctionEntry}
   > TransactObject(Object=@n1)
   > AddSimpleCollisions(self=Get StaticMeshEditorSubsystem(), StaticMesh=@n1, ShapeType=Sphere) -> (ReturnValue=0)
   > n2 = GetSimpleCollisionFromStaticMesh(StaticMesh=@n1)
   > n3 = GetSimpleCollisionShapeCount(SimpleCollision=@n2) -> (ReturnValue=0)
   > Return Node(Fix Successful="integer > integer"(A=@n3))
  ]}
 MeshFix_AddSimpleCollision_WithParameters {nodes [
   n1 = MeshFix_AddSimpleCollision_WithParameters!() -> (then, Mesh:Static Mesh Object Reference, ShapeType:EScriptCollisionShapeType Enum=Box) {is=title K2Node_FunctionEntry}
   > TransactObject(Object=@n1)
   > AddSimpleCollisions(self=Get StaticMeshEditorSubsystem(), StaticMesh=@n1, ShapeType=@n1.ShapeType) -> (ReturnValue=0)
   > n2 = GetSimpleCollisionFromStaticMesh(StaticMesh=@n1)
   > n3 = GetSimpleCollisionShapeCount(SimpleCollision=@n2) -> (ReturnValue=0)
   > Return Node(Fix Successful="integer > integer"(A=@n3))
  ]}
 MeshTest_CheckCollisions {nodes [
   n1 = MeshTest_CheckCollisions!() -> (then, StaticMesh:Static Mesh Object Reference) {is=title K2Node_FunctionEntry}
   > n2 = GetSimpleCollisionFromStaticMesh(StaticMesh=@n1)
   > n3 = GetSimpleCollisionShapeCount(SimpleCollision=@n2) -> (ReturnValue=0)
   > n4 = Branch(Condition="integer > integer"(A=@n3))
   > Return Node!(execute, Has Simple Collisions:bool=true)
   Return Node(execute=@n4.else, Has Simple Collisions=false)
   Format Text(Format="LOCTEXT(\"Warning : \\r\\nMesh does not have LODs\")")
  ]}
 MeshTest_CheckDensity {nodes [
   n1 = MeshTest_CheckDensity!() -> (then, Mesh:Static Mesh Object Reference) {is=title K2Node_FunctionEntry}
   n2 = GetNumVertices!(self:Static Mesh Object Reference=@n1, LODIndex:int=0) -> (ReturnValue:int=0)
   > n3 = Branch(Condition="integer > integer"(A=@n2, B=Get MaxVertices() -> (MaxVertices:int=0)))
   > Return Node!(execute, Has Proper Density:bool=true, Vertex Count:int=@n2)
   Return Node(execute=@n3.else, Has Proper Density=false, Vertex Count=@n2)
   n4 = Get Mesh()
   GetMinLODForQualityLevels!(self:Static Mesh Object Reference=@n4) -> (QualityLevelMinimumLODs:Map of EPerQualityLevels Enums to Integers=Low, Default:int=0)
   GetMinimumLODForQualityLevels!(self:Static Mesh Object Reference=@n4) -> (QualityLevelMinimumLODs:Map of Names to Integers)
   Format Text(Format="LOCTEXT(\"Warning : \\r\\nMesh has too many Vertices. \\r\\n\\r\\nPerformance may be negatively impacted.\")")
  ]}
 MeshTest_CheckLODs {nodes [
   n1 = MeshTest_CheckLODs!() -> (then, StaticMesh:Static Mesh Object Reference) {is=title K2Node_FunctionEntry}
   > n2 = Branch(Condition="integer > integer"(A=GetNumLODs(self=@n1), B=1))
   > Return Node!(execute, Has LODs:bool=true)
   Return Node(execute=@n2.else, Has LODs=false)
   Filter by Class!(execute, TargetArray:Array of Object References, ObjectClass:Object Class Reference, FilterType:EEditorScriptingFilterType Enum) -> (then, ReturnValue:Array of Object References)
   Format Text(Format="LOCTEXT(\"Warning : \\r\\nMesh does not have LODs\")")
  ]}
 TextureFix_Compression_GrayscaleSRGB {nodes [
   n1 = TextureFix_Compression_GrayscaleSRGB!() -> (then, Texture:Texture 2D Object Reference) {is=title K2Node_FunctionEntry}
   > TransactObject(Object=@n1)
   > Set SRGB(SRGB=false, self=@n1)
   > Return Node!(execute, Fix Successful:bool=NOT Boolean(A=Get SRGB!(self:Texture Object Reference=@n1) -> (SRGB:bool)))
   n2 = Branch(Condition="Equal (Enum)"(A:TextureCompressionSettings Enum=Get CompressionSettings(), B:TextureCompressionSettings Enum=TC_Grayscale))
   Return Node(execute=@n2.else, Fix Successful=false)
  ]}
 TextureFix_Compression_SetToBC7 {nodes [
   n1 = TextureFix_Compression_SetToBC7!() -> (then, Texture:Texture 2D Object Reference) {is=title K2Node_FunctionEntry}
   > TransactObject(Object=@n1)
   > Set CompressionSettings(CompressionSettings=TC_BC7, self=@n1)
   > Return Node(Fix Successful="Equal (Enum)"(A=Get CompressionSettings(self=@n1), B=TC_BC7))
  ]}
 TextureTest_Compression_GrayscaleSRGB {nodes [
   n1 = TextureTest_Compression_GrayscaleSRGB!() -> (then, Texture:Texture 2D Object Reference, Texture Array:Map of Texture 2D Object References to S Asset Errors Asset Health Structures) {is=title K2Node_FunctionEntry}
   n2 = Get CompressionSettings(self=@n1)
   > n3 = Branch(Condition=AND Boolean(A="Equal (Enum)"(A=@n2, B=TC_Grayscale), B=Get SRGB(self=@n1)))
   > Return Node!(
    execute
    Is Ok Grayscale:bool=false
    Updated Texture Array:Map of Texture 2D Object References to S Asset Errors Asset Health Structures=Get Texture Array() -> (Texture Array:Map of Texture 2D Object References to S Asset Errors Asset Health Structures)
   )
   Return Node(execute=@n3.else, Is Ok Grayscale=true, Updated Texture Array=Get Texture Array())
   Format Text(Format="LOCTEXT(\"Warning : \\r\\nsRGB is enabled on texture with \\'Grayscale\\' compression setting.\\r\\n\\r\\nThis combination is very expensive memory wise (4x the size of a non-sRGB grayscale texture)\\r\\n\\r\\nPlease consider disabling sRGB for this texture if possible. \")")
   Get Texture() -> (Texture:Texture 2D Object Reference)
   Get Texture Array()
  ]}
 TextureTest_Compression_Uncompressed {nodes [
   n1 = TextureTest_Compression_Uncompressed!() -> (then, Texture:Texture 2D Object Reference, Texture Array:Map of Texture 2D Object References to S Asset Errors Asset Health Structures) {is=title K2Node_FunctionEntry}
   > n2 = Branch(Condition="Equal (Enum)"(A=Get CompressionSettings(self=@n1), B=TC_EditorIcon))
   > Return Node!(execute, Is Ok Compression:bool=false, Updated Texture Array=Get Texture Array())
   Return Node(execute=@n2.else, Is Ok Compression=true, Updated Texture Array=Get Texture Array())
   Format Text(Format="LOCTEXT(\"Warning : \\r\\nTexture is set to \\'Uncompressed\\' compression setting.\\r\\n\\r\\nThis is very expensive memory wise (8x the size of a default texture)\\r\\n\\r\\nPlease consider using \\'BC7\\' compression setting if quality is a concern, otherwise use \\'Default\\'.\")")
   Get Texture()
   Get Texture Array()
  ]}
 TextureTest_Compression_VectorDisplacement {nodes [
   n1 = TextureTest_Compression_VectorDisplacement!() -> (then, Texture:Texture 2D Object Reference, Texture Array:Map of Texture 2D Object References to S Asset Errors Asset Health Structures) {is=title K2Node_FunctionEntry}
   > n2 = Branch(Condition="Equal (Enum)"(A=Get CompressionSettings(self=@n1), B=TC_VectorDisplacementmap))
   > Return Node(Is Ok Compression=false, Updated Texture Array=Get Texture Array())
   Return Node(execute=@n2.else, Is Ok Compression=true, Updated Texture Array=Get Texture Array())
   Format Text(Format="LOCTEXT(\"Warning : \\r\\nTexture is set to \\'VectorDisplacementmap\\' compression setting.\\r\\n\\r\\nThis is very expensive memory wise (8x the size of a default texture) and should only be used for vector displacement applications. \\r\\n\\r\\nIf this texture isn\\'t being used as a vector displacement please consider using \\'BC7\\' compression setting if quality is a concern, otherwise use \\'Default\\'.\")")
   Get Texture()
   Get Texture Array()
  ]}
 TextureTest_GrayscaleIssue {nodes [
   n1 = TextureTest_GrayscaleIssue!() -> (then, Texture:Texture 2D Object Reference, Texture Array:Map of Texture 2D Object References to S Asset Errors Asset Health Structures) {is=title K2Node_FunctionEntry}
   n2 = Get CompressionSettings(self=@n1)
   > n3 = Branch(Condition=AND Boolean(A="Equal (Enum)"(A=@n2, B=TC_Grayscale), B=Get SRGB(self=@n1)))
   > Return Node!(execute, Is Ok Grayscale:bool=false, Updated Texture Array=Get Texture Array())
   Return Node(execute=@n3.else, Is Ok Grayscale=true, Updated Texture Array=Get Texture Array())
   Format Text(Format="LOCTEXT(\"Warning : \\r\\nsRGB is enabled on this Grayscale Texture (Grayscale Compression)\\r\\n\\r\\nThis combination is very expensive memory wise (4x the size of a non-sRGB grayscale texture)\\r\\n\\r\\nPlease consider disabling sRGB for this texture if possible. \")")
   Get Texture()
   Get Texture Array()
  ]}
]
variables []
